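# Editor utility row widget for the Stage Manager editor that picks an FX Manager ID for one indexed event entry.
# Construct fills a searchable combo box from AvaliableFXManagerList, using a throwaway "aaaa" option around setting the current selection, and forces manual text entry when the list holds at most one item.
# Edits and deletion broadcast through dispatchers; a browse button selects matching BP_MojoFxManager level actors.
# UpdateCellStatus toggles a duplicate-ID warning icon.
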EditorUtilityWidgetBlueprint /PilgrimQuickplay/StageManager/NeverCook/EditorUtilities/StageManagerEditor/FxManagerEventWindow/EWBP_FxManagerComboBoxCell
graphs [
 EventGraph {nodes [
   n1 = "On Selection Changed (SequencesDataListComboBox)"!() -> (OutputDelegate, then, SelectedItem:String, SelectionType:ESelectInfo Enum) {is=title K2Node_ComponentBoundEvent}
   n2 = "Not Equal Exactly (String)"(A=@n1.SelectedItem, B=aaaa)
   > Branch(Condition=@n2)
   > n3 = Set CurrentFxManagerSelected(CurrentFxManagerSelected:String=@n1.SelectedItem) -> (Output_Get:String)
   > Call Fx Manager Has Changed!(execute:Exec, self:EWBP Fx Manager Combo Box Cell Object Reference, Index:int=Get Index() -> (Index:int=0), SelectedFxManager:String=@n3) -> (then:Exec) {is=title K2Node_CallDelegate}
   "On Pressed (DeleteEntryButton)"() {is=title K2Node_ComponentBoundEvent}
   > Call On Deleted!(execute:Exec, self:EWBP Fx Manager Combo Box Cell Object Reference, Index:int=Get Index() -> (Index=0)) -> (then:Exec) {is=title K2Node_CallDelegate}
   "On Pressed (OpenEditorWithThemedGroup)"() {is=title K2Node_ComponentBoundEvent}
   > n4 = GetAllActorsOfClass!(
    execute
    WorldContextObject:Object Reference
    ActorClass:Actor Class Reference=/SparksStageShow/Systems/StageManager/FXManager/BP_MojoFxManager.BP_MojoFxManager_C
   ) -> (then, OutActors:Array of BP Mojo Fx Manager Object References)
   > n5 = For Each Loop(Array:Array of BP Mojo Fx Manager Object References=@n4) -> (Array Element:BP Mojo Fx Manager Object Reference)
   SetSelectedLevelActors(
    execute=@n5.Completed
    self=Get EditorActorSubsystem()
    ActorsToSelect=Get FilteredActors() -> (FilteredActors:Array of BP Mojo Fx Manager Object References)
   )
   Branch(
    execute=@n5.LoopBody
    Condition="Equal (Name)"(
     A=Get FxManagerId!(self:Stage Show Actors Object Reference=@n5) -> (FxManagerId:Name=None)
     B:String=Get CurrentFxManagerSelected() -> (CurrentFxManagerSelected:String)
    )
   )
   > Add(TargetArray:Array of BP Mojo Fx Manager Object References=Get FilteredActors(), NewItem:"BP Mojo Fx Manager Object Reference (by ref)"=@n5) {is=title K2Node_CallArrayFunction}
   "On Pressed (ChangeToList)"() {is=title K2Node_ComponentBoundEvent}
   > SetVisibility(
    self=[
     Get ChangeToManual() -> (ChangeToManual:Editor Utility Button Object Reference)
     Get SequenceDataListSizeBox() -> (SequenceDataListSizeBox:Size Box Object Reference)
    ]
    InVisibility=Visible
   )
   > SetVisibility(
    self=[
     Get ChangeToList() -> (ChangeToList:Editor Utility Button Object Reference)
     Get SequenceFxManagerSizeBox() -> (SequenceFxManagerSizeBox:Size Box Object Reference)
    ]
    InVisibility=Collapsed
   )
   > Call Fx Manager Has Changed(Index=Get Index() -> (Index=0), SelectedFxManager=Get CurrentFxManagerSelected())
   n6 = "On Text Changed (SequenceFxManagerTextBox)"!() -> (OutputDelegate, then, Text:"Text (by ref)") {is=title K2Node_ComponentBoundEvent}
   > Call Fx Manager Has Changed(Index=Get Index() -> (Index=0), SelectedFxManager="To String (Text)"(InText=@n6.Text))
   "On Pressed (ChangeToManual)"() {is=title K2Node_ComponentBoundEvent}
   > SetVisibility(self=[Get ChangeToManual(), Get SequenceDataListSizeBox()], InVisibility=Collapsed)
   > SetVisibility(self=[Get ChangeToList(), Get SequenceFxManagerSizeBox()], InVisibility=Visible)
   > "SetText (Text Box)"!(
    execute
    self:Text Box Object Reference=Get SequenceFxManagerTextBox() -> (SequenceFxManagerTextBox:Editor Utility Editable Text Box Object Reference)
    InText:Text="To Text (String)"(InString=Get CurrentFxManagerSelected())
   ) -> (then)
   ForceChangeToManual() {is=title K2Node_CustomEvent}
   > SetVisibility(self=[Get ChangeToManual(), Get ChangeToList(), Get SequenceDataListSizeBox()], InVisibility=Collapsed)
   > SetVisibility(self=Get SequenceFxManagerSizeBox(), InVisibility=Visible)
   > "SetText (Text Box)"(self=Get SequenceFxManagerTextBox(), InText="To Text (String)"(InString=Get CurrentFxManagerSelected()))
   Event Construct()
   > n7 = Sequence()
   > n8 = For Each Loop(Array:Array of Strings=Get AvaliableFXManagerList() -> (AvaliableFXManagerList:Array of Strings)) -> (Array Element:String)
   n9 = Branch(execute=@n8.Completed, Condition=Get ExistOnTheList() -> (ExistOnTheList:bool))
   AddOption!(
    execute=[@n11.then, @n9.then]
    self:"Searchable Combo Box (String) Object Reference"=Get SequencesDataListComboBox() -> (SequencesDataListComboBox:"Searchable Combo Box (String) Object Reference")
    Option:String=aaaa
   ) -> (then)
   > SetSelectedOption!(execute, self:"Searchable Combo Box (String) Object Reference"=Get SequencesDataListComboBox(), Option:String=aaaa) -> (then)
   n10 = Get CurrentFxManagerSelected()
   > SetSelectedOption(
    self=Get SequencesDataListComboBox()
    Option=Select(Option 0:String=@n10, Option 1:String=None, Index=IsEmpty(InString=@n10)) -> (ReturnValue:String)
   )
   > RemoveOption!(execute, self:"Searchable Combo Box (String) Object Reference"=Get SequencesDataListComboBox(), Option:String=aaaa) -> (then, ReturnValue:bool)
   > "SetText (Text)"(
    self=Get EventTypeName() -> (EventTypeName:Text Object Reference)
    InText=Format Text!(Format="INVTEXT(\"Index [{index}]\")", index:int=Get Index() -> (Index=0)) -> (Result)
   )
   n11 = AddOption(execute=@n9.else, self=Get SequencesDataListComboBox(), Option=Get CurrentFxManagerSelected())
   AddOption(execute=@n8.LoopBody, self=Get SequencesDataListComboBox(), Option=@n8)
   > Set ExistOnTheList(ExistOnTheList:bool="Equal Exactly (String)"(A=@n8, B=Get CurrentFxManagerSelected())) -> (Output_Get:bool)
   n12 = Get AvaliableFXManagerList()
   n13 = Length(TargetArray:Array of Strings=Get AvaliableFXManagerList())
   Branch(execute=@n7.then_1, Condition=OR Boolean(A=Is Empty!(TargetArray=@n12) -> (ReturnValue:bool), B="Equal (Integer)"(A=@n13, B=1)))
   > ForceChangeToManual() {is=title K2Node_CallFunction}
   AND Boolean(A=@n2, B="Not Equal Exactly (String)"(A=@n1.SelectedItem, B=Get CurrentFxManagerSelected()))
   Event PreConstruct()
   Get SequencesDataListComboBox()
  ]}
 FxManagerHasChanged {nodes [FxManagerHasChanged!() -> (then, Index:int, SelectedFxManager:String) {is=title K2Node_FunctionEntry}]}
 OnDeleted {nodes [OnDeleted!() -> (then, Index:int) {is=title K2Node_FunctionEntry}]}
 UpdateCellStatus {nodes [
   n1 = UpdateCellStatus!() -> (then, IsDuplicated:bool=~) {is=title K2Node_FunctionEntry}
   > SetVisibility(
    self=Get NameDuplicatedWarningImage() -> (NameDuplicatedWarningImage:Image Object Reference)
    InVisibility=Select(Option 0:ESlateVisibility Enum=Collapsed, Option 1:ESlateVisibility Enum=Visible, Index=@n1) -> (ReturnValue:ESlateVisibility Enum)
   )
  ]}
]
variables [
 AvaliableFXManagerList {type="array<string>", default []}
 CurrentFxManagerSelected {type=string, default=""}
 ExistOnTheList {type=boolean}
 FilteredActors {type="array<BP_MojoFxManager_C>", default []}
 FxManagerHasChanged {default=()}
 Index {type=integer}
 OnDeleted {default=()}
]
widgets {
 info {rootWidgetClass=&Border, widgetCount=25}
 widgets [
  Border_0 {widgetClassPath=&Border, widget=Border_0, properties {brushColor=<rgba 0.011 0.011 0.011 1>, padding=<ltrb 0 0.5 0 0.5>}}
  Overlay_3 {widgetClassPath=&Overlay, parent=Border_0, widget=Overlay_3, slot {class=BorderSlot, properties {padding=<ltrb 0 0.5 0 0.5>}}}
  Image_68 {
   widgetClassPath=&Image
   parent=Overlay_3
   widget=Image_68
   properties {visibility=Collapsed, colorAndOpacity=<rgba 0.011 0.011 0.011 1>, renderTransform {angle=0, scale=<1 1>, shear=<0 0>, translation=<-4 0>}}
   slot {class=OverlaySlot, properties {horizontalAlignment=HAlign_Fill, verticalAlignment=VAlign_Fill, padding=<ltrb 0 -2 -5 -2>}}
  }
  Border_1 {
   widgetClassPath=&Border
   parent=Overlay_3
   widget=Border_1
   properties {brushColor=<rgba 0.0284 0.0284 0.0284 1>, padding=<ltrb 20 2 4 2>}
   slot {class=OverlaySlot, properties {horizontalAlignment=HAlign_Fill, verticalAlignment=VAlign_Fill}}
  }
  Overlay_2 {widgetClassPath=&Overlay, parent=Border_1, widget=Overlay_2, slot {class=BorderSlot, properties {padding=<ltrb 20 2 4 2>}}}
  HorizontalBox_32 {
   widgetClassPath=&HorizontalBox
   parent=Overlay_2
   widget=HorizontalBox_32
   slot {class=OverlaySlot, properties {horizontalAlignment=HAlign_Fill, verticalAlignment=VAlign_Fill}}
  }
  SizeBox_0 {
   widgetClassPath=&SizeBox
   parent=HorizontalBox_32
   widget=SizeBox_0
   properties {
    bOverride_MaxDesiredWidth=true
    bOverride_WidthOverride=true
    clipping=ClipToBounds
    maxDesiredWidth=201
    minDesiredWidth=200.3114
    widthOverride=201
   }
   slot {class=HorizontalBoxSlot, properties {verticalAlignment=VAlign_Center, padding=<ltrb 0 0 5 0>, size {sizeRule=Fill, value=1}}}
  }
  EventTypeName {
   widgetClassPath=&TextBlock
   bIsVariable=true
   parent=SizeBox_0
   widget=EventTypeName
   properties {
    text="Index [ 1 ]: "
    colorAndOpacity {colorUseRule=UseColor_Specified, specifiedColor=<rgba 0.6562 0.6562 0.6562 1>}
    font {
     bForceMonospaced=true
     bMaterialIsStencil=false
     fontMaterial=null
     fontObject=&HeadingNow
     letterSpacing=0
     monospacedWidth=0.9
     size=7.5
     skewAmount=0
     typefaceFontName=75 Medium
     outlineSettings {
      bApplyOutlineToDropShadows=false
      bMiteredCorners=false
      bSeparateFillAlpha=false
      outlineMaterial=null
      outlineSize=0
      outlineColor=<rgba 0 0 0 1>
     }
    }
   }
   slot {class=SizeBoxSlot}
  }
  Spacer_59 {widgetClassPath=&Spacer, parent=HorizontalBox_32, widget=Spacer_59, properties {size=<30 1>}, slot {class=HorizontalBoxSlot}}
  Image_95 {
   widgetClassPath=&Image
   parent=HorizontalBox_32
   widget=Image_95
   properties {
    brush {
     bIsDynamicallyLoaded=false
     drawAs=Image
     imageType=NoImage
     mirroring=NoMirror
     resourceName=None
     resourceObject=null
     tiling=NoTile
     imageSize=<2 32>
     margin=<ltrb 0 0 0 0>
     outlineSettings {
      bUseBrushTransparency=false
      roundingType=HalfHeightRadius
      width=0
      color {colorUseRule=UseColor_Specified, specifiedColor=<rgba 0 0 0 0>}
      cornerRadii=<0 0 0 0>
     }
     tintColor {colorUseRule=UseColor_Specified, specifiedColor=<rgba 1 1 1 1>}
     uVRegion {bIsValid=false, max=<0 0>, min=<0 0>}
    }
    colorAndOpacity=<rgba 0.0103 0.0103 0.0103 1>
   }
   slot {class=HorizontalBoxSlot, properties {padding=<ltrb 10 -2 10 -2>}}
  }
  SequenceDataListSizeBox {
   widgetClassPath=&SizeBox
   bIsVariable=true
   parent=HorizontalBox_32
   widget=SequenceDataListSizeBox
   properties {bOverride_MinDesiredWidth=true, minDesiredWidth=200, widthOverride=200}
   slot {class=HorizontalBoxSlot, properties {verticalAlignment=VAlign_Center, size {sizeRule=Fill, value=1}}}
  }
  SequencesDataListComboBox {
   widgetClassPath=/Script/FortniteEditor.SearchableStringComboBox
   bIsVariable=true
   parent=SequenceDataListSizeBox
   widget=SequencesDataListComboBox
   properties {
    enableGamepadNavigationMode=false
    selectedOption=Sequence Type List
    font {
     bForceMonospaced=false
     bMaterialIsStencil=false
     fontMaterial=null
     fontObject=&HeadingNow
     letterSpacing=0
     monospacedWidth=1
     size=7.5
     skewAmount=0
     typefaceFontName=75 Medium
     outlineSettings {
      bApplyOutlineToDropShadows=false
      bMiteredCorners=false
      bSeparateFillAlpha=false
      outlineMaterial=null
      outlineSize=0
      outlineColor=<rgba 0 0 0 1>
     }
    }
    foregroundColor {colorUseRule=UseColor_Specified, specifiedColor=<rgba 0.5271 0.5271 0.5271 1>}
    itemStyle {
     bUseParentRowBrush=false
     activeBrush {
      bIsDynamicallyLoaded=false
      drawAs=Image
      imageType=NoImage
      mirroring=NoMirror
      resourceName=None
      resourceObject=null
      tiling=NoTile
      imageSize=<0 0>
      margin=<ltrb 0 0 0 0>
      outlineSettings {
       bUseBrushTransparency=false
       roundingType=HalfHeightRadius
       width=0
       color {colorUseRule=UseColor_Specified, specifiedColor=<rgba 0 0 0 0>}
       cornerRadii=<0 0 0 0>
      }
      tintColor {colorUseRule=UseColor_Specified, specifiedColor=<rgba 0 0.162 0.7454 1>}
      uVRegion {bIsValid=false, max=<0 0>, min=<0 0>}
     }
     activeHighlightedBrush {
      bIsDynamicallyLoaded=false
      drawAs=Image
      imageType=NoImage
      mirroring=NoMirror
      resourceName=None
      resourceObject=null
      tiling=NoTile
      imageSize=<0 0>
      margin=<ltrb 0 0 0 0>
      outlineSettings {
       bUseBrushTransparency=false
       roundingType=HalfHeightRadius
       width=0
       color {colorUseRule=UseColor_Specified, specifiedColor=<rgba 0 0 0 0>}
       cornerRadii=<0 0 0 0>
      }
      tintColor {colorUseRule=UseColor_Specified, specifiedColor=<rgba 0.0252 0.0319 0.0423 1>}
      uVRegion {bIsValid=false, max=<0 0>, min=<0 0>}
     }
     activeHoveredBrush {
      bIsDynamicallyLoaded=false
      drawAs=Image
      imageType=NoImage
      mirroring=NoMirror
      resourceName=None
      resourceObject=null
      tiling=NoTile
      imageSize=<0 0>
      margin=<ltrb 0 0 0 0>
      outlineSettings {
       bUseBrushTransparency=false
       roundingType=HalfHeightRadius
       width=0
       color {colorUseRule=UseColor_Specified, specifiedColor=<rgba 0 0 0 0>}
       cornerRadii=<0 0 0 0>
      }
      tintColor {colorUseRule=UseColor_Specified, specifiedColor=<rgba 0 0.162 0.7454 1>}
      uVRegion {bIsValid=false, max=<0 0>, min=<0 0>}
     }
     dropIndicator_Above {
      bIsDynamicallyLoaded=false
      drawAs=Box
      imageType=FullColor
      mirroring=NoMirror
      resourceName=../../../Engine/Content/Slate/Common/DropZoneIndicator_Above.png
      resourceObject=null
      tiling=NoTile
      imageSize=<0 0>
      margin=<ltrb 0.625 0.625 0 0>
      outlineSettings {
       bUseBrushTransparency=false
       roundingType=HalfHeightRadius
       width=0
       color {colorUseRule=UseColor_Specified, specifiedColor=<rgba 0 0 0 0>}
       cornerRadii=<0 0 0 0>
      }
      tintColor {colorUseRule=UseColor_Specified, specifiedColor=<rgba 0.0091 0.3515 0.9647 1>}
      uVRegion {bIsValid=false, max=<0 0>, min=<0 0>}
     }
     dropIndicator_Below {
      bIsDynamicallyLoaded=false
      drawAs=Box
      imageType=FullColor
      mirroring=NoMirror
      resourceName=../../../Engine/Content/Slate/Common/DropZoneIndicator_Below.png
      resourceObject=null
      tiling=NoTile
      imageSize=<0 0>
      margin=<ltrb 0.625 0 0 0.625>
      outlineSettings {
       bUseBrushTransparency=false
       roundingType=HalfHeightRadius
       width=0
       color {colorUseRule=UseColor_Specified, specifiedColor=<rgba 0 0 0 0>}
       cornerRadii=<0 0 0 0>
      }
      tintColor {colorUseRule=UseColor_Specified, specifiedColor=<rgba 0.0091 0.3515 0.9647 1>}
      uVRegion {bIsValid=false, max=<0 0>, min=<0 0>}
     }
     dropIndicator_Onto {
      bIsDynamicallyLoaded=false
      drawAs=Box
      imageType=FullColor
      mirroring=NoMirror
      resourceName=../../../Engine/Content/Slate/Common/DropZoneIndicator_Onto.png
      resourceObject=null
      tiling=NoTile
      imageSize=<0 0>
      margin=<ltrb 0.25 0.25 0.25 0.25>
      outlineSettings {
       bUseBrushTransparency=false
       roundingType=HalfHeightRadius
       width=0
       color {colorUseRule=UseColor_Specified, specifiedColor=<rgba 0 0 0 0>}
       cornerRadii=<0 0 0 0>
      }
      tintColor {colorUseRule=UseColor_Specified, specifiedColor=<rgba 0.0091 0.3515 0.9647 1>}
      uVRegion {bIsValid=false, max=<0 0>, min=<0 0>}
     }
     evenRowBackgroundBrush {
      bIsDynamicallyLoaded=false
      drawAs=Image
      imageType=NoImage
      mirroring=NoMirror
      resourceName=None
      resourceObject=null
      tiling=NoTile
      imageSize=<0 0>
      margin=<ltrb 0 0 0 0>
      outlineSettings {
       bUseBrushTransparency=false
       roundingType=HalfHeightRadius
       width=0
       color {colorUseRule=UseColor_Specified, specifiedColor=<rgba 0 0 0 0>}
       cornerRadii=<0 0 0 0>
      }
      tintColor {colorUseRule=UseColor_Specified, specifiedColor=<rgba 0.0103 0.0103 0.0103 1>}
      uVRegion {bIsValid=false, max=<0 0>, min=<0 0>}
     }
     evenRowBackgroundHoveredBrush {
      bIsDynamicallyLoaded=false
      drawAs=Image
      imageType=NoImage
      mirroring=NoMirror
      resourceName=None
      resourceObject=null
      tiling=NoTile
      imageSize=<0 0>
      margin=<ltrb 0 0 0 0>
      outlineSettings {
       bUseBrushTransparency=false
       roundingType=HalfHeightRadius
       width=0
       color {colorUseRule=UseColor_Specified, specifiedColor=<rgba 0 0 0 0>}
       cornerRadii=<0 0 0 0>
      }
      tintColor {colorUseRule=UseColor_Specified, specifiedColor=<rgba 0.0176 0.0176 0.0176 1>}
      uVRegion {bIsValid=false, max=<0 0>, min=<0 0>}
     }
     inactiveBrush {
      bIsDynamicallyLoaded=false
      drawAs=Image
      imageType=NoImage
      mirroring=NoMirror
      resourceName=None
      resourceObject=null
      tiling=NoTile
      imageSize=<0 0>
      margin=<ltrb 0 0 0 0>
      outlineSettings {
       bUseBrushTransparency=false
       roundingType=HalfHeightRadius
       width=0
       color {colorUseRule=UseColor_Specified, specifiedColor=<rgba 0 0 0 0>}
       cornerRadii=<0 0 0 0>
      }
      tintColor {colorUseRule=UseColor_Specified, specifiedColor=<rgba 0.0513 0.0953 0.159 1>}
      uVRegion {bIsValid=false, max=<0 0>, min=<0 0>}
     }
     inactiveHighlightedBrush {
      bIsDynamicallyLoaded=false
      drawAs=Image
      imageType=NoImage
      mirroring=NoMirror
      resourceName=None
      resourceObject=null
      tiling=NoTile
      imageSize=<0 0>
      margin=<ltrb 0 0 0 0>
      outlineSettings {
       bUseBrushTransparency=false
       roundingType=HalfHeightRadius
       width=0
       color {colorUseRule=UseColor_Specified, specifiedColor=<rgba 0 0 0 0>}
       cornerRadii=<0 0 0 0>
      }
      tintColor {colorUseRule=UseColor_Specified, specifiedColor=<rgba 0.0252 0.0319 0.0423 1>}
      uVRegion {bIsValid=false, max=<0 0>, min=<0 0>}
     }
     inactiveHoveredBrush {
      bIsDynamicallyLoaded=false
      drawAs=Image
      imageType=NoImage
      mirroring=NoMirror
      resourceName=None
      resourceObject=null
      tiling=NoTile
      imageSize=<0 0>
      margin=<ltrb 0 0 0 0>
      outlineSettings {
       bUseBrushTransparency=false
       roundingType=HalfHeightRadius
       width=0
       color {colorUseRule=UseColor_Specified, specifiedColor=<rgba 0 0 0 0>}
       cornerRadii=<0 0 0 0>
      }
      tintColor {colorUseRule=UseColor_Specified, specifiedColor=<rgba 0.0513 0.0953 0.159 1>}
      uVRegion {bIsValid=false, max=<0 0>, min=<0 0>}
     }
     oddRowBackgroundBrush {
      bIsDynamicallyLoaded=false
      drawAs=Image
      imageType=NoImage
      mirroring=NoMirror
      resourceName=None
      resourceObject=null
      tiling=NoTile
      imageSize=<0 0>
      margin=<ltrb 0 0 0 0>
      outlineSettings {
       bUseBrushTransparency=false
       roundingType=HalfHeightRadius
       width=0
       color {colorUseRule=UseColor_Specified, specifiedColor=<rgba 0 0 0 0>}
       cornerRadii=<0 0 0 0>
      }
      tintColor {colorUseRule=UseColor_Specified, specifiedColor=<rgba 0.0103 0.0103 0.0103 1>}
      uVRegion {bIsValid=false, max=<0 0>, min=<0 0>}
     }
     oddRowBackgroundHoveredBrush {
      bIsDynamicallyLoaded=false
      drawAs=Image
      imageType=NoImage
      mirroring=NoMirror
      resourceName=None
      resourceObject=null
      tiling=NoTile
      imageSize=<0 0>
      margin=<ltrb 0 0 0 0>
      outlineSettings {
       bUseBrushTransparency=false
       roundingType=HalfHeightRadius
       width=0
       color {colorUseRule=UseColor_Specified, specifiedColor=<rgba 0 0 0 0>}
       cornerRadii=<0 0 0 0>
      }
      tintColor {colorUseRule=UseColor_Specified, specifiedColor=<rgba 0.0176 0.0176 0.0176 1>}
      uVRegion {bIsValid=false, max=<0 0>, min=<0 0>}
     }
     parentRowBackgroundBrush {
      bIsDynamicallyLoaded=false
      drawAs=Image
      imageType=NoImage
      mirroring=NoMirror
      resourceName=None
      resourceObject=null
      tiling=NoTile
      imageSize=<32 32>
      margin=<ltrb 0 0 0 0>
      outlineSettings {
       bUseBrushTransparency=false
       roundingType=HalfHeightRadius
       width=0
       color {colorUseRule=UseColor_Specified, specifiedColor=<rgba 0 0 0 0>}
       cornerRadii=<0 0 0 0>
      }
      tintColor {colorUseRule=UseColor_Specified, specifiedColor=<rgba 1 1 1 1>}
      uVRegion {bIsValid=false, max=<0 0>, min=<0 0>}
     }
     parentRowBackgroundHoveredBrush {
      bIsDynamicallyLoaded=false
      drawAs=Image
      imageType=NoImage
      mirroring=NoMirror
      resourceName=None
      resourceObject=null
      tiling=NoTile
      imageSize=<32 32>
      margin=<ltrb 0 0 0 0>
      outlineSettings {
       bUseBrushTransparency=false
       roundingType=HalfHeightRadius
       width=0
       color {colorUseRule=UseColor_Specified, specifiedColor=<rgba 0 0 0 0>}
       cornerRadii=<0 0 0 0>
      }
      tintColor {colorUseRule=UseColor_Specified, specifiedColor=<rgba 1 1 1 1>}
      uVRegion {bIsValid=false, max=<0 0>, min=<0 0>}
     }
     selectedTextColor {colorUseRule=UseColor_Specified, specifiedColor=<rgba 1 1 1 1>}
     selectorFocusedBrush {
      bIsDynamicallyLoaded=false
      drawAs=Border
      imageType=FullColor
      mirroring=NoMirror
      resourceName=../../../Engine/Content/Slate/Common/Selector.png
      resourceObject=null
      tiling=Both
      imageSize=<0 0>
      margin=<ltrb 0.25 0.25 0.25 0.25>
      outlineSettings {
       bUseBrushTransparency=false
       roundingType=HalfHeightRadius
       width=0
       color {colorUseRule=UseColor_Specified, specifiedColor=<rgba 0 0 0 0>}
       cornerRadii=<0 0 0 0>
      }
      tintColor {colorUseRule=UseColor_Specified, specifiedColor=<rgba 0.701 0.225 0.003 1>}
      uVRegion {bIsValid=false, max=<0 0>, min=<0 0>}
     }
     textColor {colorUseRule=UseColor_Specified, specifiedColor=<rgba 0.5271 0.5271 0.5271 1>}
    }
    widgetStyle {
     comboButtonStyle {
      downArrowAlign=VAlign_Center
      buttonStyle {
       clickedSlateSound {resourceObject=null}
       disabled {
        bIsDynamicallyLoaded=false
        drawAs=RoundedBox
        imageType=NoImage
        mirroring=NoMirror
        resourceName=None
        resourceObject=null
        tiling=NoTile
        imageSize=<0 0>
        margin=<ltrb 0 0 0 0>
        outlineSettings {
         bUseBrushTransparency=false
         roundingType=FixedRadius
         width=1
         color {colorUseRule=UseColor_ColorTable, specifiedColor=<rgba 0 0 0 0>}
         cornerRadii=<4 4 4 4>
        }
        tintColor {colorUseRule=UseColor_Specified, specifiedColor=<rgba 0.0395 0.0395 0.0395 1>}
        uVRegion {bIsValid=false, max=<0 0>, min=<0 0>}
       }
       disabledForeground {colorUseRule=UseColor_ColorTable, specifiedColor=<rgba 0 0 0 0>}
       hovered {
        bIsDynamicallyLoaded=false
        drawAs=RoundedBox
        imageType=NoImage
        mirroring=NoMirror
        resourceName=None
        resourceObject=null
        tiling=NoTile
        imageSize=<0 0>
        margin=<ltrb 0 0 0 0>
        outlineSettings {
         bUseBrushTransparency=false
         roundingType=FixedRadius
         width=1
         color {colorUseRule=UseColor_ColorTable, specifiedColor=<rgba 0 0 0 0>}
         cornerRadii=<4 4 4 4>
        }
        tintColor {colorUseRule=UseColor_Specified, specifiedColor=<rgba 0.0048 0.0048 0.0048 1>}
        uVRegion {bIsValid=false, max=<0 0>, min=<0 0>}
       }
       hoveredForeground {colorUseRule=UseColor_ColorTable, specifiedColor=<rgba 0 0 0 0>}
       hoveredSlateSound {resourceObject=null}
       normal {
        bIsDynamicallyLoaded=false
        drawAs=RoundedBox
        imageType=NoImage
        mirroring=NoMirror
        resourceName=None
        resourceObject=null
        tiling=NoTile
        imageSize=<0 0>
        margin=<ltrb 0 0 0 0>
        outlineSettings {
         bUseBrushTransparency=false
         roundingType=FixedRadius
         width=1
         color {colorUseRule=UseColor_ColorTable, specifiedColor=<rgba 0 0 0 0>}
         cornerRadii=<4 4 4 4>
        }
        tintColor {colorUseRule=UseColor_Specified, specifiedColor=<rgba 0.0048 0.0048 0.0048 1>}
        uVRegion {bIsValid=false, max=<0 0>, min=<0 0>}
       }
       normalForeground {colorUseRule=UseColor_ColorTable, specifiedColor=<rgba 0 0 0 0>}
       normalPadding=<ltrb 8 1 8 1>
       pressed {
        bIsDynamicallyLoaded=false
        drawAs=RoundedBox
        imageType=NoImage
        mirroring=NoMirror
        resourceName=None
        resourceObject=null
        tiling=NoTile
        imageSize=<0 0>
        margin=<ltrb 0 0 0 0>
        outlineSettings {
         bUseBrushTransparency=false
         roundingType=FixedRadius
         width=1
         color {colorUseRule=UseColor_ColorTable, specifiedColor=<rgba 0 0 0 0>}
         cornerRadii=<4 4 4 4>
        }
        tintColor {colorUseRule=UseColor_Specified, specifiedColor=<rgba 0.0048 0.0048 0.0048 1>}
        uVRegion {bIsValid=false, max=<0 0>, min=<0 0>}
       }
       pressedForeground {colorUseRule=UseColor_ColorTable, specifiedColor=<rgba 0 0 0 0>}
       pressedPadding=<ltrb 8 2 8 0>
       pressedSlateSound {resourceObject=null}
      }
      contentPadding=<ltrb 0 0 0 0>
      downArrowImage {
       bIsDynamicallyLoaded=false
       drawAs=Image
       imageType=Vector
       mirroring=NoMirror
       resourceName=../../../Engine/Content/Slate/Starship/CoreWidgets/ComboBox/wide-chevron-down.svg
       resourceObject=null
       tiling=NoTile
       imageSize=<16 16>
       margin=<ltrb 0 0 0 0>
       outlineSettings {
        bUseBrushTransparency=false
        roundingType=HalfHeightRadius
        width=0
        color {colorUseRule=UseColor_Specified, specifiedColor=<rgba 0 0 0 0>}
        cornerRadii=<0 0 0 0>
       }
       tintColor {colorUseRule=UseColor_Specified, specifiedColor=<rgba 1 1 1 1>}
       uVRegion {bIsValid=false, max=<0 0>, min=<0 0>}
      }
      downArrowPadding=<ltrb 2 2 2 2>
      menuBorderBrush {
       bIsDynamicallyLoaded=false
       drawAs=RoundedBox
       imageType=NoImage
       mirroring=NoMirror
       resourceName=None
       resourceObject=null
       tiling=NoTile
       imageSize=<0 0>
       margin=<ltrb 0 0 0 0>
       outlineSettings {
        bUseBrushTransparency=false
        roundingType=FixedRadius
        width=1
        color {colorUseRule=UseColor_ColorTable, specifiedColor=<rgba 0 0 0 0>}
        cornerRadii=<0 0 0 0>
       }
       tintColor {colorUseRule=UseColor_Specified, specifiedColor=<rgba 0.0395 0.0395 0.0395 1>}
       uVRegion {bIsValid=false, max=<0 0>, min=<0 0>}
      }
      menuBorderPadding=<ltrb 0 0 0 0>
      shadowColorAndOpacity=<rgba 0 0 0 1>
      shadowOffset=<0 0>
     }
     contentPadding=<ltrb 0 0 0 0>
     menuRowPadding=<ltrb 8 3 8 3>
     pressedSlateSound {resourceObject=null}
     selectionChangeSlateSound {resourceObject=null}
    }
   }
   slot {class=SizeBoxSlot}
  }
  SequenceFxManagerSizeBox {
   widgetClassPath=&SizeBox
   bIsVariable=true
   parent=HorizontalBox_32
   widget=SequenceFxManagerSizeBox
   properties {bOverride_MinDesiredWidth=true, minDesiredWidth=200, visibility=Collapsed}
   slot {class=HorizontalBoxSlot, properties {verticalAlignment=VAlign_Center, size {sizeRule=Fill, value=1}}}
  }
  SequenceFxManagerTextBox {
   widgetClassPath=&EditorUtilityEditableTextBox
   bIsVariable=true
   parent=SequenceFxManagerSizeBox
   widget=SequenceFxManagerTextBox
   slot {class=SizeBoxSlot}
  }
  NameDuplicatedWarningImage {
   widgetClassPath=&Image
   bIsVariable=true
   parent=HorizontalBox_32
   widget=NameDuplicatedWarningImage
   properties {
    toolTipText=Id is duplicated please choose another one.
    visibility=Collapsed
    brush {
     bIsDynamicallyLoaded=false
     drawAs=Image
     imageType=FullColor
     mirroring=NoMirror
     resourceName=None
     resourceObject=/PilgrimQuickplay/StageManager/NeverCook/EditorUtilities/StageManagerEditor/Textures/T_SparksEditorUI_IconLibrary_Warning_64
     tiling=NoTile
     imageSize=<16 16>
     margin=<ltrb 0 0 0 0>
     outlineSettings {
      bUseBrushTransparency=false
      roundingType=HalfHeightRadius
      width=0
      color {colorUseRule=UseColor_Specified, specifiedColor=<rgba 0 0 0 0>}
      cornerRadii=<0 0 0 0>
     }
     tintColor {colorUseRule=UseColor_Specified, specifiedColor=<rgba 1 1 1 1>}
     uVRegion {bIsValid=false, max=<0 0>, min=<0 0>}
    }
    colorAndOpacity=<rgba 0.9531 0.8592 0 1>
   }
   slot {class=HorizontalBoxSlot, properties {horizontalAlignment=HAlign_Left, verticalAlignment=VAlign_Center, padding=<ltrb 10 0 0 2>}}
  }
  OpenEditorWithThemedGroup {
   widgetClassPath=&EditorUtilityButton
   parent=HorizontalBox_32
   widget=OpenEditorWithThemedGroup
   properties {
    toolTipText=Browse to asset in Content Browser
    widgetStyle {
     clickedSlateSound {resourceObject=null}
     disabled {
      bIsDynamicallyLoaded=false
      drawAs=RoundedBox
      imageType=NoImage
      mirroring=NoMirror
      resourceName=None
      resourceObject=/Game/UI/Foundation/Textures/Buttons/T-Button-Square
      tiling=NoTile
      imageSize=<32 32>
      margin=<ltrb 0.5 0.5 0.5 0.5>
      outlineSettings {
       bUseBrushTransparency=false
       roundingType=FixedRadius
       width=0
       color {colorUseRule=UseColor_ColorTable, specifiedColor=<rgba 0 0 0 0>}
       cornerRadii=<4 4 4 4>
      }
      tintColor {colorUseRule=UseColor_Specified, specifiedColor=<rgba 0.0395 0.0395 0.0395 1>}
      uVRegion {bIsValid=false, max=<0 0>, min=<0 0>}
     }
     disabledForeground {colorUseRule=UseColor_ColorTable, specifiedColor=<rgba 0 0 0 0>}
     hovered {
      bIsDynamicallyLoaded=false
      drawAs=RoundedBox
      imageType=NoImage
      mirroring=NoMirror
      resourceName=None
      resourceObject=&T-Button-Square
      tiling=NoTile
      imageSize=<32 32>
      margin=<ltrb 0.5 0.5 0.5 0.5>
      outlineSettings {
       bUseBrushTransparency=false
       roundingType=FixedRadius
       width=0
       color {colorUseRule=UseColor_ColorTable, specifiedColor=<rgba 0 0 0 0>}
       cornerRadii=<4 4 4 4>
      }
      tintColor {colorUseRule=UseColor_Specified, specifiedColor=<rgba 0 0.3021 0.0241 1>}
      uVRegion {bIsValid=false, max=<0 0>, min=<0 0>}
     }
     hoveredForeground {colorUseRule=UseColor_ColorTable, specifiedColor=<rgba 0 0 0 0>}
     hoveredSlateSound {resourceObject=null}
     normal {
      bIsDynamicallyLoaded=false
      drawAs=RoundedBox
      imageType=NoImage
      mirroring=NoMirror
      resourceName=None
      resourceObject=&T-Button-Square
      tiling=NoTile
      imageSize=<32 32>
      margin=<ltrb 0.5 0.5 0.5 0.5>
      outlineSettings {
       bUseBrushTransparency=false
       roundingType=FixedRadius
       width=0
       color {colorUseRule=UseColor_ColorTable, specifiedColor=<rgba 0 0 0 0>}
       cornerRadii=<4 4 4 4>
      }
      tintColor {colorUseRule=UseColor_Specified, specifiedColor=<rgba 0.0071 0.7594 0 1>}
      uVRegion {bIsValid=false, max=<0 0>, min=<0 0>}
     }
     normalForeground {colorUseRule=UseColor_ColorTable, specifiedColor=<rgba 0 0 0 0>}
     normalPadding=<ltrb 12 1.5 12 1.5>
     pressed {
      bIsDynamicallyLoaded=false
      drawAs=RoundedBox
      imageType=NoImage
      mirroring=NoMirror
      resourceName=None
      resourceObject=&T-Button-Square
      tiling=NoTile
      imageSize=<32 32>
      margin=<ltrb 0.5 0.5 0.5 0.5>
      outlineSettings {
       bUseBrushTransparency=false
       roundingType=FixedRadius
       width=0
       color {colorUseRule=UseColor_ColorTable, specifiedColor=<rgba 0 0 0 0>}
       cornerRadii=<4 4 4 4>
      }
      tintColor {colorUseRule=UseColor_Specified, specifiedColor=<rgba 0.3119 0.7917 0.2715 1>}
      uVRegion {bIsValid=false, max=<0 0>, min=<0 0>}
     }
     pressedForeground {colorUseRule=UseColor_ColorTable, specifiedColor=<rgba 0 0 0 0>}
     pressedPadding=<ltrb 12 2.5 12 0.5>
     pressedSlateSound {resourceObject=null}
    }
   }
   slot {class=HorizontalBoxSlot, properties {horizontalAlignment=HAlign_Left, verticalAlignment=VAlign_Center, padding=<ltrb 10 0 0 0>}}
  }
  ExternalIcon {
   widgetClassPath=&Image
   parent=OpenEditorWithThemedGroup
   widget=ExternalIcon
   properties {
    brush {
     bIsDynamicallyLoaded=false
     drawAs=Image
     imageType=FullColor
     mirroring=NoMirror
     resourceName=None
     resourceObject=/PilgrimQuickplay/StageManager/NeverCook/EditorUtilities/StageManagerEditor/Textures/T_SparksEditorUI_IconMagnifying_Glass_32
     tiling=NoTile
     imageSize=<17 17>
     margin=<ltrb 0 0 0 0>
     outlineSettings {
      bUseBrushTransparency=false
      roundingType=HalfHeightRadius
      width=0
      color {colorUseRule=UseColor_Specified, specifiedColor=<rgba 0 0 0 0>}
      cornerRadii=<0 0 0 0>
     }
     tintColor {colorUseRule=UseColor_Specified, specifiedColor=<rgba 1 1 1 1>}
     uVRegion {bIsValid=false, max=<0 0>, min=<0 0>}
    }
    renderTransform {angle=0, scale=<-1 1>, shear=<0 0>, translation=<0 0>}
   }
   slot {class=ButtonSlot, properties {padding=<ltrb -2 0 -2 0>}}
  }
  DeleteEntryButton {
   widgetClassPath=&EditorUtilityButton
   parent=HorizontalBox_32
   widget=DeleteEntryButton
   properties {
    clipping=ClipToBounds
    widgetStyle {
     clickedSlateSound {resourceObject=null}
     disabled {
      bIsDynamicallyLoaded=false
      drawAs=RoundedBox
      imageType=NoImage
      mirroring=NoMirror
      resourceName=None
      resourceObject=null
      tiling=NoTile
      imageSize=<20 20>
      margin=<ltrb 0 0 0 0>
      outlineSettings {
       bUseBrushTransparency=false
       roundingType=FixedRadius
       width=0
       color {colorUseRule=UseColor_ColorTable, specifiedColor=<rgba 0 0 0 0>}
       cornerRadii=<4 4 4 4>
      }
      tintColor {colorUseRule=UseColor_Specified, specifiedColor=<rgba 0.0395 0.0395 0.0395 1>}
      uVRegion {bIsValid=false, max=<0 0>, min=<0 0>}
     }
     disabledForeground {colorUseRule=UseColor_ColorTable, specifiedColor=<rgba 0 0 0 0>}
     hovered {
      bIsDynamicallyLoaded=false
      drawAs=RoundedBox
      imageType=NoImage
      mirroring=NoMirror
      resourceName=None
      resourceObject=null
      tiling=NoTile
      imageSize=<32 32>
      margin=<ltrb 0 0 0 0>
      outlineSettings {
       bUseBrushTransparency=false
       roundingType=FixedRadius
       width=0
       color {colorUseRule=UseColor_ColorTable, specifiedColor=<rgba 0 0 0 0>}
       cornerRadii=<4 4 4 4>
      }
      tintColor {colorUseRule=UseColor_Specified, specifiedColor=<rgba 1 0.1252 0.253 1>}
      uVRegion {bIsValid=false, max=<0 0>, min=<0 0>}
     }
     hoveredForeground {colorUseRule=UseColor_ColorTable, specifiedColor=<rgba 0 0 0 0>}
     hoveredSlateSound {resourceObject=null}
     normal {
      bIsDynamicallyLoaded=false
      drawAs=RoundedBox
      imageType=NoImage
      mirroring=NoMirror
      resourceName=None
      resourceObject=null
      tiling=NoTile
      imageSize=<20 20>
      margin=<ltrb 0 0 0 0>
      outlineSettings {
       bUseBrushTransparency=false
       roundingType=FixedRadius
       width=0
       color {colorUseRule=UseColor_ColorTable, specifiedColor=<rgba 0 0 0 0>}
       cornerRadii=<4 4 4 4>
      }
      tintColor {colorUseRule=UseColor_Specified, specifiedColor=<rgba 0.775 0 0.0304 1>}
      uVRegion {bIsValid=false, max=<0 0>, min=<0 0>}
     }
     normalForeground {colorUseRule=UseColor_ColorTable, specifiedColor=<rgba 0 0 0 0>}
     normalPadding=<ltrb 12 1.5 12 1.5>
     pressed {
      bIsDynamicallyLoaded=false
      drawAs=RoundedBox
      imageType=NoImage
      mirroring=NoMirror
      resourceName=None
      resourceObject=null
      tiling=NoTile
      imageSize=<20 20>
      margin=<ltrb 0 0 0 0>
      outlineSettings {
       bUseBrushTransparency=false
       roundingType=FixedRadius
       width=0
       color {colorUseRule=UseColor_ColorTable, specifiedColor=<rgba 0 0 0 0>}
       cornerRadii=<4 4 4 4>
      }
      tintColor {colorUseRule=UseColor_Specified, specifiedColor=<rgba 0.2396 0 0 1>}
      uVRegion {bIsValid=false, max=<0 0>, min=<0 0>}
     }
     pressedForeground {colorUseRule=UseColor_ColorTable, specifiedColor=<rgba 0 0 0 0>}
     pressedPadding=<ltrb 12 2.5 12 0.5>
     pressedSlateSound {resourceObject=null}
    }
   }
   slot {class=HorizontalBoxSlot, properties {horizontalAlignment=HAlign_Left, verticalAlignment=VAlign_Center, padding=<ltrb 10 0 0 0>}}
  }
  Image_20 {
   widgetClassPath=&Image
   parent=DeleteEntryButton
   widget=Image_20
   properties {
    visibility=SelfHitTestInvisible
    brush {
     bIsDynamicallyLoaded=false
     drawAs=Image
     imageType=FullColor
     mirroring=NoMirror
     resourceName=None
     resourceObject=/PilgrimQuickplay/StageManager/NeverCook/EditorUtilities/StageManagerEditor/Textures/T_SparksEditorUI_TrashCan_64
     tiling=NoTile
     imageSize=<17 17>
     margin=<ltrb 0 0 0 0>
     outlineSettings {
      bUseBrushTransparency=false
      roundingType=HalfHeightRadius
      width=0
      color {colorUseRule=UseColor_Specified, specifiedColor=<rgba 0 0 0 0>}
      cornerRadii=<0 0 0 0>
     }
     tintColor {colorUseRule=UseColor_Specified, specifiedColor=<rgba 1 1 1 1>}
     uVRegion {bIsValid=false, max=<0 0>, min=<0 0>}
    }
   }
   slot {class=ButtonSlot, properties {padding=<ltrb -5 0 -5 0>}}
  }
  ChangeToManual {
   widgetClassPath=&EditorUtilityButton
   parent=HorizontalBox_32
   widget=ChangeToManual
   properties {
    clipping=ClipToBounds
    toolTipText=Instead choose from list set manually the fxmanager id
    widgetStyle {
     clickedSlateSound {resourceObject=null}
     disabled {
      bIsDynamicallyLoaded=false
      drawAs=RoundedBox
      imageType=NoImage
      mirroring=NoMirror
      resourceName=None
      resourceObject=null
      tiling=NoTile
      imageSize=<32 32>
      margin=<ltrb 0.5 0.5 0.5 0.5>
      outlineSettings {
       bUseBrushTransparency=false
       roundingType=FixedRadius
       width=0
       color {colorUseRule=UseColor_ColorTable, specifiedColor=<rgba 0 0 0 0>}
       cornerRadii=<4 4 4 4>
      }
      tintColor {colorUseRule=UseColor_Specified, specifiedColor=<rgba 0.0395 0.0395 0.0395 1>}
      uVRegion {bIsValid=false, max=<0 0>, min=<0 0>}
     }
     disabledForeground {colorUseRule=UseColor_ColorTable, specifiedColor=<rgba 0 0 0 0>}
     hovered {
      bIsDynamicallyLoaded=false
      drawAs=RoundedBox
      imageType=NoImage
      mirroring=NoMirror
      resourceName=None
      resourceObject=null
      tiling=NoTile
      imageSize=<32 32>
      margin=<ltrb 0 0 0 0>
      outlineSettings {
       bUseBrushTransparency=false
       roundingType=FixedRadius
       width=0
       color {colorUseRule=UseColor_ColorTable, specifiedColor=<rgba 0 0 0 0>}
       cornerRadii=<4 4 4 4>
      }
      tintColor {colorUseRule=UseColor_Specified, specifiedColor=<rgba 0.1126 0.7031 0 1>}
      uVRegion {bIsValid=false, max=<0 0>, min=<0 0>}
     }
     hoveredForeground {colorUseRule=UseColor_ColorTable, specifiedColor=<rgba 0 0 0 0>}
     hoveredSlateSound {resourceObject=null}
     normal {
      bIsDynamicallyLoaded=false
      drawAs=RoundedBox
      imageType=NoImage
      mirroring=NoMirror
      resourceName=None
      resourceObject=null
      tiling=NoTile
      imageSize=<32 32>
      margin=<ltrb 0.5 0.5 0.5 0.5>
      outlineSettings {
       bUseBrushTransparency=false
       roundingType=FixedRadius
       width=0
       color {colorUseRule=UseColor_ColorTable, specifiedColor=<rgba 0 0 0 0>}
       cornerRadii=<4 4 4 4>
      }
      tintColor {colorUseRule=UseColor_Specified, specifiedColor=<rgba 0.0771 0.6167 0.1289 1>}
      uVRegion {bIsValid=false, max=<0 0>, min=<0 0>}
     }
     normalForeground {colorUseRule=UseColor_ColorTable, specifiedColor=<rgba 0 0 0 0>}
     normalPadding=<ltrb 12 1.5 12 1.5>
     pressed {
      bIsDynamicallyLoaded=false
      drawAs=RoundedBox
      imageType=NoImage
      mirroring=NoMirror
      resourceName=None
      resourceObject=null
      tiling=NoTile
      imageSize=<32 32>
      margin=<ltrb 0.5 0.5 0.5 0.5>
      outlineSettings {
       bUseBrushTransparency=false
       roundingType=FixedRadius
       width=0
       color {colorUseRule=UseColor_ColorTable, specifiedColor=<rgba 0 0 0 0>}
       cornerRadii=<4 4 4 4>
      }
      tintColor {colorUseRule=UseColor_Specified, specifiedColor=<rgba 0.1806 0.4583 0.1572 1>}
      uVRegion {bIsValid=false, max=<0 0>, min=<0 0>}
     }
     pressedForeground {colorUseRule=UseColor_ColorTable, specifiedColor=<rgba 0 0 0 0>}
     pressedPadding=<ltrb 12 2.5 12 0.5>
     pressedSlateSound {resourceObject=null}
    }
   }
   slot {
    class=HorizontalBoxSlot
    properties {horizontalAlignment=HAlign_Left, verticalAlignment=VAlign_Center, padding=<ltrb 5 0 0 0>, size {sizeRule=Fill, value=1}}
   }
  }
  HorizontalBox_4 {
   widgetClassPath=&HorizontalBox
   parent=ChangeToManual
   widget=HorizontalBox_4
   slot {class=ButtonSlot, properties {horizontalAlignment=HAlign_Left}}
  }
  TextBlock_25 {
   widgetClassPath=&TextBlock
   parent=HorizontalBox_4
   widget=TextBlock_25
   properties {
    text=Manual Input
    font {
     bForceMonospaced=false
     bMaterialIsStencil=false
     fontMaterial=null
     fontObject=&HeadingNow
     letterSpacing=0
     monospacedWidth=1
     size=7.5
     skewAmount=0
     typefaceFontName=75 Medium
     outlineSettings {
      bApplyOutlineToDropShadows=false
      bMiteredCorners=false
      bSeparateFillAlpha=false
      outlineMaterial=null
      outlineSize=0
      outlineColor=<rgba 0 0 0 1>
     }
    }
   }
   slot {class=HorizontalBoxSlot, properties {verticalAlignment=VAlign_Center}}
  }
  ChangeToList {
   widgetClassPath=&EditorUtilityButton
   parent=HorizontalBox_32
   widget=ChangeToList
   properties {
    clipping=ClipToBounds
    toolTipText="Return to the list mode that will show all avaliable fxmanagers on the level (recomended to use it when the level is loaded first)"
    visibility=Collapsed
    widgetStyle {
     clickedSlateSound {resourceObject=null}
     disabled {
      bIsDynamicallyLoaded=false
      drawAs=RoundedBox
      imageType=NoImage
      mirroring=NoMirror
      resourceName=None
      resourceObject=null
      tiling=NoTile
      imageSize=<32 32>
      margin=<ltrb 0.5 0.5 0.5 0.5>
      outlineSettings {
       bUseBrushTransparency=false
       roundingType=FixedRadius
       width=0
       color {colorUseRule=UseColor_ColorTable, specifiedColor=<rgba 0 0 0 0>}
       cornerRadii=<4 4 4 4>
      }
      tintColor {colorUseRule=UseColor_Specified, specifiedColor=<rgba 0.0395 0.0395 0.0395 1>}
      uVRegion {bIsValid=false, max=<0 0>, min=<0 0>}
     }
     disabledForeground {colorUseRule=UseColor_ColorTable, specifiedColor=<rgba 0 0 0 0>}
     hovered {
      bIsDynamicallyLoaded=false
      drawAs=RoundedBox
      imageType=NoImage
      mirroring=NoMirror
      resourceName=None
      resourceObject=null
      tiling=NoTile
      imageSize=<32 32>
      margin=<ltrb 0 0 0 0>
      outlineSettings {
       bUseBrushTransparency=false
       roundingType=FixedRadius
       width=0
       color {colorUseRule=UseColor_ColorTable, specifiedColor=<rgba 0 0 0 0>}
       cornerRadii=<4 4 4 4>
      }
      tintColor {colorUseRule=UseColor_Specified, specifiedColor=<rgba 0.1126 0.7031 0 1>}
      uVRegion {bIsValid=false, max=<0 0>, min=<0 0>}
     }
     hoveredForeground {colorUseRule=UseColor_ColorTable, specifiedColor=<rgba 0 0 0 0>}
     hoveredSlateSound {resourceObject=null}
     normal {
      bIsDynamicallyLoaded=false
      drawAs=RoundedBox
      imageType=NoImage
      mirroring=NoMirror
      resourceName=None
      resourceObject=null
      tiling=NoTile
      imageSize=<32 32>
      margin=<ltrb 0.5 0.5 0.5 0.5>
      outlineSettings {
       bUseBrushTransparency=false
       roundingType=FixedRadius
       width=0
       color {colorUseRule=UseColor_ColorTable, specifiedColor=<rgba 0 0 0 0>}
       cornerRadii=<4 4 4 4>
      }
      tintColor {colorUseRule=UseColor_Specified, specifiedColor=<rgba 0.0771 0.6167 0.1289 1>}
      uVRegion {bIsValid=false, max=<0 0>, min=<0 0>}
     }
     normalForeground {colorUseRule=UseColor_ColorTable, specifiedColor=<rgba 0 0 0 0>}
     normalPadding=<ltrb 12 1.5 12 1.5>
     pressed {
      bIsDynamicallyLoaded=false
      drawAs=RoundedBox
      imageType=NoImage
      mirroring=NoMirror
      resourceName=None
      resourceObject=null
      tiling=NoTile
      imageSize=<32 32>
      margin=<ltrb 0.5 0.5 0.5 0.5>
      outlineSettings {
       bUseBrushTransparency=false
       roundingType=FixedRadius
       width=0
       color {colorUseRule=UseColor_ColorTable, specifiedColor=<rgba 0 0 0 0>}
       cornerRadii=<4 4 4 4>
      }
      tintColor {colorUseRule=UseColor_Specified, specifiedColor=<rgba 0.1806 0.4583 0.1572 1>}
      uVRegion {bIsValid=false, max=<0 0>, min=<0 0>}
     }
     pressedForeground {colorUseRule=UseColor_ColorTable, specifiedColor=<rgba 0 0 0 0>}
     pressedPadding=<ltrb 12 2.5 12 0.5>
     pressedSlateSound {resourceObject=null}
    }
   }
   slot {
    class=HorizontalBoxSlot
    properties {horizontalAlignment=HAlign_Left, verticalAlignment=VAlign_Center, padding=<ltrb 5 0 0 0>, size {sizeRule=Fill, value=1}}
   }
  }
  HorizontalBox {
   widgetClassPath=&HorizontalBox
   parent=ChangeToList
   widget=HorizontalBox
   slot {class=ButtonSlot, properties {horizontalAlignment=HAlign_Left}}
  }
  TextBlock {
   widgetClassPath=&TextBlock
   parent=HorizontalBox
   widget=TextBlock
   properties {
    text=List Input
    font {
     bForceMonospaced=false
     bMaterialIsStencil=false
     fontMaterial=null
     fontObject=&HeadingNow
     letterSpacing=0
     monospacedWidth=1
     size=7.5
     skewAmount=0
     typefaceFontName=75 Medium
     outlineSettings {
      bApplyOutlineToDropShadows=false
      bMiteredCorners=false
      bSeparateFillAlpha=false
      outlineMaterial=null
      outlineSize=0
      outlineColor=<rgba 0 0 0 1>
     }
    }
   }
   slot {class=HorizontalBoxSlot, properties {verticalAlignment=VAlign_Center}}
  }
 ]
}
