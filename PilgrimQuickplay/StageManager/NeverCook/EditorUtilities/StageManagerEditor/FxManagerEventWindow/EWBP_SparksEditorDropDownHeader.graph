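# Editor utility widget for the Sparks stage manager editor: a collapsible header for array-style lists, with a title, element count, and add and clear buttons.
# The dropdown button toggles ContentVerticalBox visibility and swaps arrow images; PreConstruct sets the title from HeaderTitle.
# Add and trash buttons only fire OnAddNewElement and OnClearAllElements dispatchers, leaving changes to listeners.
# AddContentChild, RemoveContentChild and ClearContentChildren edit the box and refresh the pluralized count label.
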EditorUtilityWidgetBlueprint /PilgrimQuickplay/StageManager/NeverCook/EditorUtilities/StageManagerEditor/FxManagerEventWindow/EWBP_SparksEditorDropDownHeader
graphs [
 AddContentChild {nodes [
   n1 = AddContentChild!() -> (then, Content:Widget Object Reference) {is=title K2Node_FunctionEntry}
   > n2 = AddChild(self=Get ContentVerticalBox() -> (ContentVerticalBox:Vertical Box Object Reference), Content=@n1)
   > UpdateChildCount()
   > Return Node!(execute, ReturnValue:Panel Slot Object Reference=@n2)
  ]}
 ClearContentChildren {nodes [
   ClearContentChildren() {is=title K2Node_FunctionEntry}
   > ClearChildren(self=Get ContentVerticalBox())
   > UpdateChildCount()
  ]}
 EventGraph {nodes [
   "On Pressed (DropDownButton)"() {is=title K2Node_ComponentBoundEvent}
   > n1 = Branch(Condition=Get IsOpenned() -> (IsOpenned:bool))
   > Set IsOpenned(IsOpenned:bool=false) -> (Output_Get:bool)
   > SetVisibility(self=Get ArrowUp() -> (ArrowUp:Image Object Reference), InVisibility=Visible)
   > SetVisibility(self=[Get ArrowDown() -> (ArrowDown:Image Object Reference), Get ContentVerticalBox()], InVisibility=Collapsed)
   Set IsOpenned(execute=@n1.else, IsOpenned=true)
   > SetVisibility(self=Get ArrowUp(), InVisibility=Collapsed)
   > SetVisibility(self=[Get ArrowDown(), Get ContentVerticalBox()], InVisibility=Visible)
   "On Pressed (AddMIDITextMapButton)"() {is=title K2Node_ComponentBoundEvent}
   > Call On Add New Element!(execute:Exec, self:EWBP Sparks Editor Drop Down Header Object Reference) -> (then:Exec) {is=title K2Node_CallDelegate}
   "On Pressed (DeleteMoodButton)"() {is=title K2Node_ComponentBoundEvent}
   > Call On Clear All Elements!(execute:Exec, self:EWBP Sparks Editor Drop Down Header Object Reference) -> (then:Exec) {is=title K2Node_CallDelegate}
   Event PreConstruct()
   > "SetText (Text)"(self=Get TitleName() -> (TitleName:Text Object Reference), InText=Get HeaderTitle() -> (HeaderTitle:Text))
   Event Construct()
  ]}
 OnAddNewElement {nodes [OnAddNewElement() {is=title K2Node_FunctionEntry}]}
 OnClearAllElements {nodes [OnClearAllElements() {is=title K2Node_FunctionEntry}]}
 RemoveContentChild {nodes [
   n1 = RemoveContentChild!() -> (then, Content:Widget Object Reference) {is=title K2Node_FunctionEntry}
   > n2 = RemoveChild(self=Get ContentVerticalBox(), Content=@n1)
   > UpdateChildCount()
   > Return Node(ReturnValue:bool=@n2)
  ]}
 UpdateChildCount {nodes [
   UpdateChildCount() {is=title K2Node_FunctionEntry}
   n1 = GetChildrenCount!(self:Panel Widget Object Reference=Get ContentVerticalBox()) -> (ReturnValue:int=0)
   n2 = "integer > integer"(A=@n1, B=1)
   > "SetText (Text)"(
    self=Get ChildCountText() -> (ChildCountText:Text Object Reference)
    InText=Format Text!(
     Format="LOCTEXT(\"{count} Array element{s}\")"
     count:int=@n1
     s:Text=Select(Option 0:Text, Option 1:Text="LOCTEXT(\"s\")", Index=@n2) -> (ReturnValue:Text)
    ) -> (Result)
   )
  ]}
]
variables [
 HeaderTitle {type=string, default=HeaderName}
 IsOpenned {type=boolean, default=true}
 OnAddNewElement {default=()}
 OnClearAllElements {default=()}
]
widgets {
 info {rootWidgetClass=&Border, widgetCount=20}
 widgets [
  Border_1 {
   widgetClassPath=&Border
   widget=Border_1
   properties {
    verticalAlignment=VAlign_Top
    background {
     bIsDynamicallyLoaded=false
     drawAs=NoDrawType
     imageType=NoImage
     mirroring=NoMirror
     resourceName=None
     resourceObject=null
     tiling=NoTile
     imageSize=<32 32>
     margin=<ltrb 0 0 0 0>
     outlineSettings {
      bUseBrushTransparency=false
      roundingType=HalfHeightRadius
      width=0
      color {colorUseRule=UseColor_Specified, specifiedColor=<rgba 0 0 0 0>}
      cornerRadii=<0 0 0 0>
     }
     tintColor {colorUseRule=UseColor_Specified, specifiedColor=<rgba 1 1 1 1>}
     uVRegion {bIsValid=false, max=<0 0>, min=<0 0>}
    }
    padding=<ltrb 0 2 0 2>
   }
  }
  Overlay_0 {
   widgetClassPath=&Overlay
   parent=Border_1
   widget=Overlay_0
   slot {class=BorderSlot, properties {verticalAlignment=VAlign_Top, padding=<ltrb 0 2 0 2>}}
  }
  Image_151 {
   widgetClassPath=&Image
   parent=Overlay_0
   widget=Image_151
   properties {visibility=Collapsed, colorAndOpacity=<rgba 0.0097 0.0097 0.0097 1>}
   slot {class=OverlaySlot, properties {horizontalAlignment=HAlign_Fill, padding=<ltrb -2 -2 -2 -2>}}
  }
  VerticalBox_0 {
   widgetClassPath=&VerticalBox
   parent=Overlay_0
   widget=VerticalBox_0
   slot {class=OverlaySlot, properties {horizontalAlignment=HAlign_Fill, verticalAlignment=VAlign_Fill}}
  }
  Border_43 {
   widgetClassPath=&Border
   parent=VerticalBox_0
   widget=Border_43
   properties {verticalAlignment=VAlign_Top, brushColor=<rgba 0.0176 0.0176 0.0176 1>, padding=<ltrb 0 5 5 5>}
   slot {class=VerticalBoxSlot}
  }
  HorizontalBox_0 {
   widgetClassPath=&HorizontalBox
   parent=Border_43
   widget=HorizontalBox_0
   slot {class=BorderSlot, properties {verticalAlignment=VAlign_Top, padding=<ltrb 0 5 5 5>}}
  }
  DropDownButton {
   widgetClassPath=&EditorUtilityButton
   parent=HorizontalBox_0
   widget=DropDownButton
   properties {widgetStyle {
     clickedSlateSound {resourceObject=null}
     disabled {
      bIsDynamicallyLoaded=false
      drawAs=RoundedBox
      imageType=NoImage
      mirroring=NoMirror
      resourceName=None
      resourceObject=/Game/UI/Foundation/Textures/Buttons/T-Button-Square
      tiling=NoTile
      imageSize=<32 32>
      margin=<ltrb 0.5 0.5 0.5 0.5>
      outlineSettings {
       bUseBrushTransparency=false
       roundingType=FixedRadius
       width=0
       color {colorUseRule=UseColor_ColorTable, specifiedColor=<rgba 0 0 0 0>}
       cornerRadii=<4 4 4 4>
      }
      tintColor {colorUseRule=UseColor_Specified, specifiedColor=<rgba 0.0395 0.0395 0.0395 1>}
      uVRegion {bIsValid=false, max=<0 0>, min=<0 0>}
     }
     disabledForeground {colorUseRule=UseColor_ColorTable, specifiedColor=<rgba 0 0 0 0>}
     hovered {
      bIsDynamicallyLoaded=false
      drawAs=RoundedBox
      imageType=NoImage
      mirroring=NoMirror
      resourceName=None
      resourceObject=&T-Button-Square
      tiling=NoTile
      imageSize=<15 15>
      margin=<ltrb 0.5 0.5 0.5 0.5>
      outlineSettings {
       bUseBrushTransparency=false
       roundingType=FixedRadius
       width=0
       color {colorUseRule=UseColor_ColorTable, specifiedColor=<rgba 0 0 0 0>}
       cornerRadii=<4 4 4 4>
      }
      tintColor {colorUseRule=UseColor_Specified, specifiedColor=<rgba 0.7344 0.4651 0 1>}
      uVRegion {bIsValid=false, max=<0 0>, min=<0 0>}
     }
     hoveredForeground {colorUseRule=UseColor_ColorTable, specifiedColor=<rgba 0 0 0 0>}
     hoveredSlateSound {resourceObject=null}
     normal {
      bIsDynamicallyLoaded=false
      drawAs=RoundedBox
      imageType=NoImage
      mirroring=NoMirror
      resourceName=None
      resourceObject=&T-Button-Square
      tiling=NoTile
      imageSize=<15 15>
      margin=<ltrb 0.5 0.5 0.5 0.5>
      outlineSettings {
       bUseBrushTransparency=false
       roundingType=FixedRadius
       width=0
       color {colorUseRule=UseColor_ColorTable, specifiedColor=<rgba 0 0 0 0>}
       cornerRadii=<4 4 4 4>
      }
      tintColor {colorUseRule=UseColor_Specified, specifiedColor=<rgba 1 0.6333 0 1>}
      uVRegion {bIsValid=false, max=<0 0>, min=<0 0>}
     }
     normalForeground {colorUseRule=UseColor_ColorTable, specifiedColor=<rgba 0 0 0 0>}
     normalPadding=<ltrb 2 1.5 2 1.5>
     pressed {
      bIsDynamicallyLoaded=false
      drawAs=RoundedBox
      imageType=NoImage
      mirroring=NoMirror
      resourceName=None
      resourceObject=&T-Button-Square
      tiling=NoTile
      imageSize=<15 15>
      margin=<ltrb 0.5 0.5 0.5 0.5>
      outlineSettings {
       bUseBrushTransparency=false
       roundingType=FixedRadius
       width=0
       color {colorUseRule=UseColor_ColorTable, specifiedColor=<rgba 0 0 0 0>}
       cornerRadii=<4 4 4 4>
      }
      tintColor {colorUseRule=UseColor_Specified, specifiedColor=<rgba 0.6458 0.409 0 1>}
      uVRegion {bIsValid=false, max=<0 0>, min=<0 0>}
     }
     pressedForeground {colorUseRule=UseColor_ColorTable, specifiedColor=<rgba 0 0 0 0>}
     pressedPadding=<ltrb 2 1.5 2 1.5>
     pressedSlateSound {resourceObject=null}
    }}
   slot {
    class=HorizontalBoxSlot
    properties {horizontalAlignment=HAlign_Left, verticalAlignment=VAlign_Center, padding=<ltrb 5 0 5 0>, size {sizeRule=Automatic, value=0.2}}
   }
  }
  Overlay_3 {
   widgetClassPath=&Overlay
   parent=DropDownButton
   widget=Overlay_3
   slot {class=ButtonSlot, properties {horizontalAlignment=HAlign_Fill, verticalAlignment=VAlign_Fill, padding=<ltrb 0 0 0 0>}}
  }
  ArrowDown {
   widgetClassPath=&Image
   bIsVariable=true
   parent=Overlay_3
   widget=ArrowDown
   properties {brush {
     bIsDynamicallyLoaded=false
     drawAs=Image
     imageType=NoImage
     mirroring=NoMirror
     resourceName=None
     resourceObject=/PilgrimQuickplay/StageManager/NeverCook/EditorUtilities/StageManagerEditor/Textures/T_EditorDownArrow
     tiling=NoTile
     imageSize=<15 15>
     margin=<ltrb 0 0 0 0>
     outlineSettings {
      bUseBrushTransparency=false
      roundingType=HalfHeightRadius
      width=0
      color {colorUseRule=UseColor_Specified, specifiedColor=<rgba 0 0 0 0>}
      cornerRadii=<0 0 0 0>
     }
     tintColor {colorUseRule=UseColor_Specified, specifiedColor=<rgba 1 1 1 1>}
     uVRegion {bIsValid=false, max=<0 0>, min=<0 0>}
    }}
   slot {class=OverlaySlot}
  }
  ArrowUp {
   widgetClassPath=&Image
   bIsVariable=true
   parent=Overlay_3
   widget=ArrowUp
   properties {
    visibility=Collapsed
    brush {
     bIsDynamicallyLoaded=false
     drawAs=Image
     imageType=NoImage
     mirroring=NoMirror
     resourceName=None
     resourceObject=&T_EditorDownArrow
     tiling=NoTile
     imageSize=<15 15>
     margin=<ltrb 0 0 0 0>
     outlineSettings {
      bUseBrushTransparency=false
      roundingType=HalfHeightRadius
      width=0
      color {colorUseRule=UseColor_Specified, specifiedColor=<rgba 0 0 0 0>}
      cornerRadii=<0 0 0 0>
     }
     tintColor {colorUseRule=UseColor_Specified, specifiedColor=<rgba 1 1 1 1>}
     uVRegion {bIsValid=false, max=<0 0>, min=<0 0>}
    }
    renderTransform {angle=-90, scale=<1 1>, shear=<0 0>, translation=<0 0>}
   }
   slot {class=OverlaySlot}
  }
  SizeBox_0 {
   widgetClassPath=&SizeBox
   parent=HorizontalBox_0
   widget=SizeBox_0
   properties {bOverride_MinDesiredWidth=true, clipping=ClipToBounds, maxDesiredWidth=20.3452, minDesiredWidth=190, widthOverride=94.8393}
   slot {class=HorizontalBoxSlot, properties {verticalAlignment=VAlign_Center, size {sizeRule=Fill, value=1}}}
  }
  TitleName {
   widgetClassPath=&TextBlock
   bIsVariable=true
   parent=SizeBox_0
   widget=TitleName
   properties {
    text=HeaderName
    colorAndOpacity {colorUseRule=UseColor_Specified, specifiedColor=<rgba 0.6562 0.6562 0.6562 1>}
    font {
     bForceMonospaced=false
     bMaterialIsStencil=false
     fontMaterial=null
     fontObject=&HeadingNow
     letterSpacing=0
     monospacedWidth=1
     size=7.5
     skewAmount=0
     typefaceFontName=74 Regular
     outlineSettings {
      bApplyOutlineToDropShadows=false
      bMiteredCorners=false
      bSeparateFillAlpha=false
      outlineMaterial=null
      outlineSize=0
      outlineColor=<rgba 0 0 0 1>
     }
    }
   }
   slot {class=SizeBoxSlot}
  }
  Spacer {
   widgetClassPath=&Spacer
   parent=HorizontalBox_0
   widget=Spacer
   properties {size=<36 1>}
   slot {class=HorizontalBoxSlot, properties {horizontalAlignment=HAlign_Left, size {sizeRule=Automatic, value=1.6}}}
  }
  Image_95 {
   widgetClassPath=&Image
   parent=HorizontalBox_0
   widget=Image_95
   properties {
    brush {
     bIsDynamicallyLoaded=false
     drawAs=Image
     imageType=NoImage
     mirroring=NoMirror
     resourceName=None
     resourceObject=null
     tiling=NoTile
     imageSize=<2 32>
     margin=<ltrb 0 0 0 0>
     outlineSettings {
      bUseBrushTransparency=false
      roundingType=HalfHeightRadius
      width=0
      color {colorUseRule=UseColor_Specified, specifiedColor=<rgba 0 0 0 0>}
      cornerRadii=<0 0 0 0>
     }
     tintColor {colorUseRule=UseColor_Specified, specifiedColor=<rgba 1 1 1 1>}
     uVRegion {bIsValid=false, max=<0 0>, min=<0 0>}
    }
    colorAndOpacity=<rgba 0.0103 0.0103 0.0103 1>
   }
   slot {class=HorizontalBoxSlot, properties {padding=<ltrb 10 -5 10 -5>}}
  }
  ChildCountText {
   widgetClassPath=&TextBlock
   bIsVariable=true
   parent=HorizontalBox_0
   widget=ChildCountText
   properties {
    clipping=ClipToBounds
    text=1 array element
    colorAndOpacity {colorUseRule=UseColor_Specified, specifiedColor=<rgba 0.6562 0.6562 0.6562 1>}
    font {
     bForceMonospaced=false
     bMaterialIsStencil=false
     fontMaterial=null
     fontObject=&HeadingNow
     letterSpacing=0
     monospacedWidth=1
     size=7.5
     skewAmount=0
     typefaceFontName=74 Regular
     outlineSettings {
      bApplyOutlineToDropShadows=false
      bMiteredCorners=false
      bSeparateFillAlpha=false
      outlineMaterial=null
      outlineSize=0
      outlineColor=<rgba 0 0 0 1>
     }
    }
   }
   slot {class=HorizontalBoxSlot, properties {verticalAlignment=VAlign_Center, size {sizeRule=Fill, value=1}}}
  }
  AddMIDITextMapButton {
   widgetClassPath=&EditorUtilityButton
   parent=HorizontalBox_0
   widget=AddMIDITextMapButton
   properties {
    toolTipText=Add new element
    widgetStyle {
     clickedSlateSound {resourceObject=null}
     disabled {
      bIsDynamicallyLoaded=false
      drawAs=RoundedBox
      imageType=NoImage
      mirroring=NoMirror
      resourceName=None
      resourceObject=null
      tiling=NoTile
      imageSize=<32 32>
      margin=<ltrb 0.5 0.5 0.5 0.5>
      outlineSettings {
       bUseBrushTransparency=false
       roundingType=FixedRadius
       width=0
       color {colorUseRule=UseColor_ColorTable, specifiedColor=<rgba 0 0 0 0>}
       cornerRadii=<4 4 4 4>
      }
      tintColor {colorUseRule=UseColor_Specified, specifiedColor=<rgba 0.0395 0.0395 0.0395 1>}
      uVRegion {bIsValid=false, max=<0 0>, min=<0 0>}
     }
     disabledForeground {colorUseRule=UseColor_ColorTable, specifiedColor=<rgba 0 0 0 0>}
     hovered {
      bIsDynamicallyLoaded=false
      drawAs=RoundedBox
      imageType=NoImage
      mirroring=NoMirror
      resourceName=None
      resourceObject=null
      tiling=NoTile
      imageSize=<15 15>
      margin=<ltrb 0 0 0 0>
      outlineSettings {
       bUseBrushTransparency=false
       roundingType=FixedRadius
       width=1
       color {colorUseRule=UseColor_ColorTable, specifiedColor=<rgba 0 0 0 0>}
       cornerRadii=<4 4 4 4>
      }
      tintColor {colorUseRule=UseColor_Specified, specifiedColor=<rgba 0 1 0.0916 1>}
      uVRegion {bIsValid=false, max=<0 0>, min=<0 0>}
     }
     hoveredForeground {colorUseRule=UseColor_ColorTable, specifiedColor=<rgba 0 0 0 0>}
     hoveredSlateSound {resourceObject=null}
     normal {
      bIsDynamicallyLoaded=false
      drawAs=RoundedBox
      imageType=NoImage
      mirroring=NoMirror
      resourceName=None
      resourceObject=null
      tiling=NoTile
      imageSize=<15 15>
      margin=<ltrb 0.5 0.5 0.5 0.5>
      outlineSettings {
       bUseBrushTransparency=false
       roundingType=FixedRadius
       width=0
       color {colorUseRule=UseColor_ColorTable, specifiedColor=<rgba 0 0 0 0>}
       cornerRadii=<4 4 4 4>
      }
      tintColor {colorUseRule=UseColor_Specified, specifiedColor=<rgba 0.0771 0.6167 0.1289 1>}
      uVRegion {bIsValid=false, max=<0 0>, min=<0 0>}
     }
     normalForeground {colorUseRule=UseColor_ColorTable, specifiedColor=<rgba 0 0 0 0>}
     normalPadding=<ltrb 12 1.5 12 1.5>
     pressed {
      bIsDynamicallyLoaded=false
      drawAs=RoundedBox
      imageType=NoImage
      mirroring=NoMirror
      resourceName=None
      resourceObject=null
      tiling=NoTile
      imageSize=<15 15>
      margin=<ltrb 0.5 0.5 0.5 0.5>
      outlineSettings {
       bUseBrushTransparency=false
       roundingType=FixedRadius
       width=0
       color {colorUseRule=UseColor_ColorTable, specifiedColor=<rgba 0 0 0 0>}
       cornerRadii=<4 4 4 4>
      }
      tintColor {colorUseRule=UseColor_Specified, specifiedColor=<rgba 0.1806 0.4583 0.1572 1>}
      uVRegion {bIsValid=false, max=<0 0>, min=<0 0>}
     }
     pressedForeground {colorUseRule=UseColor_ColorTable, specifiedColor=<rgba 0 0 0 0>}
     pressedPadding=<ltrb 12 2.5 12 0.5>
     pressedSlateSound {resourceObject=null}
    }
   }
   slot {
    class=HorizontalBoxSlot
    properties {horizontalAlignment=HAlign_Left, verticalAlignment=VAlign_Center, padding=<ltrb 10 0 0 0>, size {sizeRule=Automatic, value=0.1}}
   }
  }
  Image_15 {
   widgetClassPath=&Image
   parent=AddMIDITextMapButton
   widget=Image_15
   properties {brush {
     bIsDynamicallyLoaded=false
     drawAs=Image
     imageType=NoImage
     mirroring=NoMirror
     resourceName=None
     resourceObject=/PilgrimQuickplay/StageManager/NeverCook/EditorUtilities/StageManagerEditor/Textures/MI_SparksEditorUI_PlusMinus_NoBorder
     tiling=NoTile
     imageSize=<17 17>
     margin=<ltrb 0 0 0 0>
     outlineSettings {
      bUseBrushTransparency=false
      roundingType=HalfHeightRadius
      width=0
      color {colorUseRule=UseColor_Specified, specifiedColor=<rgba 0 0 0 0>}
      cornerRadii=<0 0 0 0>
     }
     tintColor {colorUseRule=UseColor_Specified, specifiedColor=<rgba 1 1 1 1>}
     uVRegion {bIsValid=false, max=<0 0>, min=<0 0>}
    }}
   slot {class=ButtonSlot, properties {padding=<ltrb -5 0 -5 0>}}
  }
  DeleteMoodButton {
   widgetClassPath=&EditorUtilityButton
   parent=HorizontalBox_0
   widget=DeleteMoodButton
   properties {
    toolTipText=Remove All Elements
    widgetStyle {
     clickedSlateSound {resourceObject=null}
     disabled {
      bIsDynamicallyLoaded=false
      drawAs=RoundedBox
      imageType=NoImage
      mirroring=NoMirror
      resourceName=None
      resourceObject=null
      tiling=NoTile
      imageSize=<20 20>
      margin=<ltrb 0 0 0 0>
      outlineSettings {
       bUseBrushTransparency=false
       roundingType=FixedRadius
       width=0
       color {colorUseRule=UseColor_ColorTable, specifiedColor=<rgba 0 0 0 0>}
       cornerRadii=<4 4 4 4>
      }
      tintColor {colorUseRule=UseColor_Specified, specifiedColor=<rgba 0.0395 0.0395 0.0395 1>}
      uVRegion {bIsValid=false, max=<0 0>, min=<0 0>}
     }
     disabledForeground {colorUseRule=UseColor_ColorTable, specifiedColor=<rgba 0 0 0 0>}
     hovered {
      bIsDynamicallyLoaded=false
      drawAs=RoundedBox
      imageType=NoImage
      mirroring=NoMirror
      resourceName=None
      resourceObject=null
      tiling=NoTile
      imageSize=<32 32>
      margin=<ltrb 0 0 0 0>
      outlineSettings {
       bUseBrushTransparency=false
       roundingType=FixedRadius
       width=0
       color {colorUseRule=UseColor_ColorTable, specifiedColor=<rgba 0 0 0 0>}
       cornerRadii=<4 4 4 4>
      }
      tintColor {colorUseRule=UseColor_Specified, specifiedColor=<rgba 1 0.1252 0.253 1>}
      uVRegion {bIsValid=false, max=<0 0>, min=<0 0>}
     }
     hoveredForeground {colorUseRule=UseColor_ColorTable, specifiedColor=<rgba 0 0 0 0>}
     hoveredSlateSound {resourceObject=null}
     normal {
      bIsDynamicallyLoaded=false
      drawAs=RoundedBox
      imageType=NoImage
      mirroring=NoMirror
      resourceName=None
      resourceObject=null
      tiling=NoTile
      imageSize=<20 20>
      margin=<ltrb 0 0 0 0>
      outlineSettings {
       bUseBrushTransparency=false
       roundingType=FixedRadius
       width=0
       color {colorUseRule=UseColor_ColorTable, specifiedColor=<rgba 0 0 0 0>}
       cornerRadii=<4 4 4 4>
      }
      tintColor {colorUseRule=UseColor_Specified, specifiedColor=<rgba 0.775 0 0.0304 1>}
      uVRegion {bIsValid=false, max=<0 0>, min=<0 0>}
     }
     normalForeground {colorUseRule=UseColor_ColorTable, specifiedColor=<rgba 0 0 0 0>}
     normalPadding=<ltrb 12 1.5 12 1.5>
     pressed {
      bIsDynamicallyLoaded=false
      drawAs=RoundedBox
      imageType=NoImage
      mirroring=NoMirror
      resourceName=None
      resourceObject=null
      tiling=NoTile
      imageSize=<20 20>
      margin=<ltrb 0 0 0 0>
      outlineSettings {
       bUseBrushTransparency=false
       roundingType=FixedRadius
       width=0
       color {colorUseRule=UseColor_ColorTable, specifiedColor=<rgba 0 0 0 0>}
       cornerRadii=<4 4 4 4>
      }
      tintColor {colorUseRule=UseColor_Specified, specifiedColor=<rgba 0.2396 0 0 1>}
      uVRegion {bIsValid=false, max=<0 0>, min=<0 0>}
     }
     pressedForeground {colorUseRule=UseColor_ColorTable, specifiedColor=<rgba 0 0 0 0>}
     pressedPadding=<ltrb 12 2.5 12 0.5>
     pressedSlateSound {resourceObject=null}
    }
   }
   slot {
    class=HorizontalBoxSlot
    properties {horizontalAlignment=HAlign_Left, verticalAlignment=VAlign_Center, padding=<ltrb 10 0 0 0>, size {sizeRule=Fill, value=1}}
   }
  }
  Image_20 {
   widgetClassPath=&Image
   parent=DeleteMoodButton
   widget=Image_20
   properties {
    visibility=SelfHitTestInvisible
    brush {
     bIsDynamicallyLoaded=false
     drawAs=Image
     imageType=FullColor
     mirroring=NoMirror
     resourceName=None
     resourceObject=/PilgrimQuickplay/StageManager/NeverCook/EditorUtilities/StageManagerEditor/Textures/T_SparksEditorUI_TrashCan_64
     tiling=NoTile
     imageSize=<17 17>
     margin=<ltrb 0 0 0 0>
     outlineSettings {
      bUseBrushTransparency=false
      roundingType=HalfHeightRadius
      width=0
      color {colorUseRule=UseColor_Specified, specifiedColor=<rgba 0 0 0 0>}
      cornerRadii=<0 0 0 0>
     }
     tintColor {colorUseRule=UseColor_Specified, specifiedColor=<rgba 1 1 1 1>}
     uVRegion {bIsValid=false, max=<0 0>, min=<0 0>}
    }
   }
   slot {class=ButtonSlot, properties {padding=<ltrb -5 0 -5 0>}}
  }
  ContentVerticalBox {
   widgetClassPath=&VerticalBox
   bIsVariable=true
   parent=VerticalBox_0
   widget=ContentVerticalBox
   slot {class=VerticalBoxSlot, properties {padding=<ltrb 30 0 0 0>}}
  }
 ]
}
