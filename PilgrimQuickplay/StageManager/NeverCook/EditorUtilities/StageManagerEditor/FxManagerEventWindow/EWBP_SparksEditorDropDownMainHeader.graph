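# Collapsible section header widget for the Sparks editor in the Stage Manager's FX event editor tools, titled from HeaderTitle.
# The drop-down button flips IsOpenned, swaps arrow images, shows or collapses the content box and broadcasts OnTogglePressed.
# Content functions add, remove or clear children, binding and unbinding single-property-view change events so OnPropertyViewChange can relay them through OnAnyPropertyChange.
# SetChildPropertyEnabledAt sets a child's enabled state and tooltip by index.
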EditorUtilityWidgetBlueprint /PilgrimQuickplay/StageManager/NeverCook/EditorUtilities/StageManagerEditor/FxManagerEventWindow/EWBP_SparksEditorDropDownMainHeader
graphs [
 AddContentChild {nodes [
   n1 = AddContentChild!() -> (then, Content:Widget Object Reference) {is=title K2Node_FunctionEntry}
   > n2 = AddChild(self=Get ContentVerticalBox() -> (ContentVerticalBox:Vertical Box Object Reference), Content=@n1)
   > n3 = Sequence()
   > n4 = Cast To EWBP_SparksEditorSinglePropertyView!(execute, Object=@n1) -> (then, CastFailed, AsEWBP Sparks Editor Single Property View:EWBP Sparks Editor Single Property View Object Reference) {is=title K2Node_DynamicCast}
   > Bind Event to On Property Changed(self:EWBP Sparks Editor Single Property View Object Reference=@n4, Delegate=Create Event()) {is=title K2Node_AddDelegate}
   Return Node!(execute=@n3.then_1, ReturnValue:Panel Slot Object Reference=@n2)
  ]}
 ClearContentChildren {nodes [
   ClearContentChildren() {is=title K2Node_FunctionEntry}
   n1 = Get ContentVerticalBox()
   > n2 = For Each Loop(Array:Array of Widget Object References=GetAllChildren(self=@n1)) -> (Array Element:Widget Object Reference)
   ClearChildren(execute=@n2.Completed, self=@n1)
   n3 = Cast To EWBP_SparksEditorSinglePropertyView(execute=@n2.LoopBody, Object=@n2)
   > Unbind Event from On Property Changed(self:EWBP Sparks Editor Single Property View Object Reference=@n3, Delegate=Create Event()) {is=title K2Node_RemoveDelegate}
  ]}
 EventGraph {nodes [
   "On Pressed (DropDownButton)"() {is=title K2Node_ComponentBoundEvent}
   > Set IsOpenned(IsOpenned:bool=NOT Boolean(A=Get IsOpenned() -> (IsOpenned:bool))) -> (Output_Get:bool)
   > SetVisibility(self=Get ArrowUp() -> (ArrowUp:Image Object Reference), InVisibility=Select(Option 0=Visible, Option 1=Collapsed, Index=Get IsOpenned()))
   > SetVisibility(
    self=[Get ContentVerticalBox(), Get ArrowDown() -> (ArrowDown:Image Object Reference)]
    InVisibility=Select(Option 0=Collapsed, Option 1=Visible, Index=Get IsOpenned())
   )
   > Call On Toggle Pressed!(execute:Exec, self:EWBP Sparks Editor Drop Down Main Header Object Reference, IsOpenned:bool=Get IsOpenned()) -> (then:Exec) {is=title K2Node_CallDelegate}
   Event PreConstruct()
   > "SetText (Text)"(self=Get TitleName() -> (TitleName:Text Object Reference), InText=Get HeaderTitle() -> (HeaderTitle:Text))
  ]}
 OnAnyPropertyChange {nodes [OnAnyPropertyChange!() -> (then, PropertyName:Name) {is=title K2Node_FunctionEntry}]}
 OnPropertyViewChange {nodes [
   n1 = OnPropertyViewChange!() -> (then, PropertyName:Name) {is=title K2Node_FunctionEntry}
   > Call On Any Property Change!(execute:Exec, self:EWBP Sparks Editor Drop Down Main Header Object Reference, PropertyName:Name=@n1) -> (then:Exec) {is=title K2Node_CallDelegate}
  ]}
 OnTogglePressed {nodes [OnTogglePressed!() -> (then, IsOpenned:bool=~) {is=title K2Node_FunctionEntry}]}
 RemoveContentChild {nodes [
   n1 = RemoveContentChild!() -> (then, Content:Widget Object Reference) {is=title K2Node_FunctionEntry}
   > n2 = Sequence()
   > n3 = Cast To EWBP_SparksEditorSinglePropertyView(Object=@n1)
   > Unbind Event from On Property Changed(self=@n3, Delegate=Create Event())
   n4 = RemoveChild(execute=@n2.then_1, self=Get ContentVerticalBox(), Content=@n1)
   > Return Node(ReturnValue:bool=@n4)
  ]}
 SetChildPropertyEnabledAt {nodes [
   n1 = SetChildPropertyEnabledAt!() -> (then, Index:int, bInIsEnabled:bool=~, InToolTipText:"Text (by ref)") {is=title K2Node_FunctionEntry}
   n2 = GetChildAt!(self:Panel Widget Object Reference=Get ContentVerticalBox(), Index:int=@n1) -> (ReturnValue:Widget Object Reference)
   > Is Valid(InputObject=@n2)
   > SetIsEnabled(self=@n2, bInIsEnabled=@n1.bInIsEnabled)
   > SetToolTipText!(execute, self:Widget Object Reference=@n2, InToolTipText:"Text (by ref)"=@n1.InToolTipText) -> (then)
  ]}
]
variables [
 HeaderTitle {type=string, default=HeaderName}
 IsOpenned {type=boolean, default=true}
 OnAnyPropertyChange {default=()}
 OnTogglePressed {default=()}
]
widgets {
 info {rootWidgetClass=&Overlay, widgetCount=13}
 widgets [
  Overlay_0 {widgetClassPath=&Overlay, widget=Overlay_0}
  Image_151 {
   widgetClassPath=&Image
   parent=Overlay_0
   widget=Image_151
   properties {
    brush {
     bIsDynamicallyLoaded=false
     drawAs=Image
     imageType=NoImage
     mirroring=NoMirror
     resourceName=None
     resourceObject=null
     tiling=NoTile
     imageSize=<32 29>
     margin=<ltrb 0 0 0 0>
     outlineSettings {
      bUseBrushTransparency=false
      roundingType=HalfHeightRadius
      width=0
      color {colorUseRule=UseColor_Specified, specifiedColor=<rgba 0 0 0 0>}
      cornerRadii=<0 0 0 0>
     }
     tintColor {colorUseRule=UseColor_Specified, specifiedColor=<rgba 1 1 1 1>}
     uVRegion {bIsValid=false, max=<0 0>, min=<0 0>}
    }
    colorAndOpacity=<rgba 0.0104 0.0104 0.0104 1>
   }
   slot {class=OverlaySlot, properties {horizontalAlignment=HAlign_Fill, padding=<ltrb 0 0 0 5>}}
  }
  VerticalBox_0 {
   widgetClassPath=&VerticalBox
   parent=Overlay_0
   widget=VerticalBox_0
   slot {class=OverlaySlot, properties {horizontalAlignment=HAlign_Fill, verticalAlignment=VAlign_Fill}}
  }
  Border_43 {
   widgetClassPath=&Border
   parent=VerticalBox_0
   widget=Border_43
   properties {verticalAlignment=VAlign_Center, brushColor=<rgba 0.0284 0.0284 0.0284 1>, padding=<ltrb 0 5 5 5>}
   slot {class=VerticalBoxSlot}
  }
  Overlay_2 {
   widgetClassPath=&Overlay
   parent=Border_43
   widget=Overlay_2
   slot {class=BorderSlot, properties {verticalAlignment=VAlign_Center, padding=<ltrb 0 5 5 5>}}
  }
  HorizontalBox_0 {
   widgetClassPath=&HorizontalBox
   parent=Overlay_2
   widget=HorizontalBox_0
   slot {class=OverlaySlot, properties {padding=<ltrb 0 0 10 0>}}
  }
  DropDownButton {
   widgetClassPath=&EditorUtilityButton
   parent=HorizontalBox_0
   widget=DropDownButton
   properties {widgetStyle {
     clickedSlateSound {resourceObject=null}
     disabled {
      bIsDynamicallyLoaded=false
      drawAs=RoundedBox
      imageType=NoImage
      mirroring=NoMirror
      resourceName=None
      resourceObject=/Game/UI/Foundation/Textures/Buttons/T-Button-Square
      tiling=NoTile
      imageSize=<32 32>
      margin=<ltrb 0.5 0.5 0.5 0.5>
      outlineSettings {
       bUseBrushTransparency=false
       roundingType=FixedRadius
       width=0
       color {colorUseRule=UseColor_ColorTable, specifiedColor=<rgba 0 0 0 0>}
       cornerRadii=<4 4 4 4>
      }
      tintColor {colorUseRule=UseColor_Specified, specifiedColor=<rgba 0.0395 0.0395 0.0395 1>}
      uVRegion {bIsValid=false, max=<0 0>, min=<0 0>}
     }
     disabledForeground {colorUseRule=UseColor_ColorTable, specifiedColor=<rgba 0 0 0 0>}
     hovered {
      bIsDynamicallyLoaded=false
      drawAs=RoundedBox
      imageType=NoImage
      mirroring=NoMirror
      resourceName=None
      resourceObject=&T-Button-Square
      tiling=NoTile
      imageSize=<15 15>
      margin=<ltrb 0.5 0.5 0.5 0.5>
      outlineSettings {
       bUseBrushTransparency=false
       roundingType=FixedRadius
       width=0
       color {colorUseRule=UseColor_ColorTable, specifiedColor=<rgba 0 0 0 0>}
       cornerRadii=<4 4 4 4>
      }
      tintColor {colorUseRule=UseColor_Specified, specifiedColor=<rgba 0.7344 0.4651 0 1>}
      uVRegion {bIsValid=false, max=<0 0>, min=<0 0>}
     }
     hoveredForeground {colorUseRule=UseColor_ColorTable, specifiedColor=<rgba 0 0 0 0>}
     hoveredSlateSound {resourceObject=null}
     normal {
      bIsDynamicallyLoaded=false
      drawAs=RoundedBox
      imageType=NoImage
      mirroring=NoMirror
      resourceName=None
      resourceObject=&T-Button-Square
      tiling=NoTile
      imageSize=<15 15>
      margin=<ltrb 0.5 0.5 0.5 0.5>
      outlineSettings {
       bUseBrushTransparency=false
       roundingType=FixedRadius
       width=0
       color {colorUseRule=UseColor_ColorTable, specifiedColor=<rgba 0 0 0 0>}
       cornerRadii=<4 4 4 4>
      }
      tintColor {colorUseRule=UseColor_Specified, specifiedColor=<rgba 1 0.6333 0 1>}
      uVRegion {bIsValid=false, max=<0 0>, min=<0 0>}
     }
     normalForeground {colorUseRule=UseColor_ColorTable, specifiedColor=<rgba 0 0 0 0>}
     normalPadding=<ltrb 2 1.5 2 1.5>
     pressed {
      bIsDynamicallyLoaded=false
      drawAs=RoundedBox
      imageType=NoImage
      mirroring=NoMirror
      resourceName=None
      resourceObject=&T-Button-Square
      tiling=NoTile
      imageSize=<15 15>
      margin=<ltrb 0.5 0.5 0.5 0.5>
      outlineSettings {
       bUseBrushTransparency=false
       roundingType=FixedRadius
       width=0
       color {colorUseRule=UseColor_ColorTable, specifiedColor=<rgba 0 0 0 0>}
       cornerRadii=<4 4 4 4>
      }
      tintColor {colorUseRule=UseColor_Specified, specifiedColor=<rgba 0.6458 0.409 0 1>}
      uVRegion {bIsValid=false, max=<0 0>, min=<0 0>}
     }
     pressedForeground {colorUseRule=UseColor_ColorTable, specifiedColor=<rgba 0 0 0 0>}
     pressedPadding=<ltrb 2 1.5 2 1.5>
     pressedSlateSound {resourceObject=null}
    }}
   slot {class=HorizontalBoxSlot, properties {horizontalAlignment=HAlign_Left, verticalAlignment=VAlign_Center, padding=<ltrb 5 0 5 0>}}
  }
  Overlay_3 {
   widgetClassPath=&Overlay
   parent=DropDownButton
   widget=Overlay_3
   slot {class=ButtonSlot, properties {horizontalAlignment=HAlign_Fill, verticalAlignment=VAlign_Fill, padding=<ltrb 0 0 0 0>}}
  }
  ArrowDown {
   widgetClassPath=&Image
   bIsVariable=true
   parent=Overlay_3
   widget=ArrowDown
   properties {brush {
     bIsDynamicallyLoaded=false
     drawAs=Image
     imageType=NoImage
     mirroring=NoMirror
     resourceName=None
     resourceObject=/PilgrimQuickplay/StageManager/NeverCook/EditorUtilities/StageManagerEditor/Textures/T_EditorDownArrow
     tiling=NoTile
     imageSize=<15 15>
     margin=<ltrb 0 0 0 0>
     outlineSettings {
      bUseBrushTransparency=false
      roundingType=HalfHeightRadius
      width=0
      color {colorUseRule=UseColor_Specified, specifiedColor=<rgba 0 0 0 0>}
      cornerRadii=<0 0 0 0>
     }
     tintColor {colorUseRule=UseColor_Specified, specifiedColor=<rgba 1 1 1 1>}
     uVRegion {bIsValid=false, max=<0 0>, min=<0 0>}
    }}
   slot {class=OverlaySlot}
  }
  ArrowUp {
   widgetClassPath=&Image
   bIsVariable=true
   parent=Overlay_3
   widget=ArrowUp
   properties {
    visibility=Collapsed
    brush {
     bIsDynamicallyLoaded=false
     drawAs=Image
     imageType=NoImage
     mirroring=NoMirror
     resourceName=None
     resourceObject=&T_EditorDownArrow
     tiling=NoTile
     imageSize=<15 15>
     margin=<ltrb 0 0 0 0>
     outlineSettings {
      bUseBrushTransparency=false
      roundingType=HalfHeightRadius
      width=0
      color {colorUseRule=UseColor_Specified, specifiedColor=<rgba 0 0 0 0>}
      cornerRadii=<0 0 0 0>
     }
     tintColor {colorUseRule=UseColor_Specified, specifiedColor=<rgba 1 1 1 1>}
     uVRegion {bIsValid=false, max=<0 0>, min=<0 0>}
    }
    renderTransform {angle=-90, scale=<1 1>, shear=<0 0>, translation=<0 0>}
   }
   slot {class=OverlaySlot}
  }
  SizeBox_0 {
   widgetClassPath=&SizeBox
   parent=HorizontalBox_0
   widget=SizeBox_0
   properties {bOverride_MinDesiredWidth=true, minDesiredWidth=300, widthOverride=200}
   slot {class=HorizontalBoxSlot, properties {horizontalAlignment=HAlign_Left, verticalAlignment=VAlign_Center, size {sizeRule=Fill, value=1}}}
  }
  TitleName {
   widgetClassPath=&TextBlock
   bIsVariable=true
   parent=SizeBox_0
   widget=TitleName
   properties {
    text=HeaderName
    colorAndOpacity {colorUseRule=UseColor_Specified, specifiedColor=<rgba 0.6562 0.6562 0.6562 1>}
    font {
     bForceMonospaced=false
     bMaterialIsStencil=false
     fontMaterial=null
     fontObject=&HeadingNow
     letterSpacing=0
     monospacedWidth=1
     size=9
     skewAmount=0
     typefaceFontName=76 Bold
     outlineSettings {
      bApplyOutlineToDropShadows=false
      bMiteredCorners=false
      bSeparateFillAlpha=false
      outlineMaterial=null
      outlineSize=0
      outlineColor=<rgba 0 0 0 1>
     }
    }
   }
   slot {class=SizeBoxSlot}
  }
  ContentVerticalBox {
   widgetClassPath=&VerticalBox
   bIsVariable=true
   parent=VerticalBox_0
   widget=ContentVerticalBox
   slot {class=VerticalBoxSlot, properties {padding=<ltrb 24 0 0 0>}}
  }
 ]
}
