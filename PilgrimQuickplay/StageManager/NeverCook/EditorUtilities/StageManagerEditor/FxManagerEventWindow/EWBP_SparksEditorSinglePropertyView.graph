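# Editor-only utility widget for the Stage Manager FX event editor that shows one editable property of an object as a styled row.
# SetObject points the embedded SinglePropertyView at a given object and property name, then writes that name into a separate text label.
# The view's built-in name is overridden with a space so the custom label stands in for it.
# Edits are re-broadcast through the OnPropertyChanged dispatcher with the property name.
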EditorUtilityWidgetBlueprint /PilgrimQuickplay/StageManager/NeverCook/EditorUtilities/StageManagerEditor/FxManagerEventWindow/EWBP_SparksEditorSinglePropertyView
graphs [
 EventGraph {nodes [
   n1 = "On Property Changed (SinglePropertyView_36)"!() -> (OutputDelegate, then, PropertyName:Name=None) {is=title K2Node_ComponentBoundEvent}
   > Call On Property Changed!(execute:Exec, self:EWBP Sparks Editor Single Property View Object Reference, PropertyName:Name=@n1.PropertyName) -> (then:Exec) {is=title K2Node_CallDelegate}
   Event PreConstruct()
   Event Construct()
   Event Tick()
  ]}
 OnPropertyChanged {nodes [OnPropertyChanged!() -> (then, PropertyName:Name) {is=title K2Node_FunctionEntry}]}
 SetObject {nodes [
   n1 = SetObject!() -> (then, NewObject:Object Reference, NewPropertyName:Name) {is=title K2Node_FunctionEntry}
   > SetObject(self=Get SinglePropertyView_36() -> (SinglePropertyView_36:Single Property View Object Reference), NewObject=@n1) {is=title K2Node_CallFunction}
   n2 = Get SinglePropertyView_36()
   > SetPropertyName!(execute, self:Single Property View Object Reference=@n2, NewPropertyName:Name=@n1.NewPropertyName) -> (then)
   > "SetText (Text)"(
    self=Get PropertyNameText() -> (PropertyNameText:Text Object Reference)
    InText="To Text (Name)"!(InName:Name=@n1.NewPropertyName) -> (ReturnValue:Text)
   )
   "To Text (Name)"(InName=GetPropertyName!(self:Single Property View Object Reference=@n2) -> (ReturnValue:Name=None))
  ]}
]
variables [
 ObjectToView {type=Object, default=/PilgrimQuickplay/StageManager/Data/Events/Pilgrim_StageFX_Event_Intro02_ColorShift}
 OnPropertyChanged {default=()}
]
widgets {
 info {rootWidgetClass=&Border, widgetCount=11}
 widgets [
  Border_1 {
   widgetClassPath=&Border
   widget=Border_1
   properties {
    background {
     bIsDynamicallyLoaded=false
     drawAs=NoDrawType
     imageType=NoImage
     mirroring=NoMirror
     resourceName=None
     resourceObject=null
     tiling=NoTile
     imageSize=<32 32>
     margin=<ltrb 0 0 0 0>
     outlineSettings {
      bUseBrushTransparency=false
      roundingType=HalfHeightRadius
      width=0
      color {colorUseRule=UseColor_Specified, specifiedColor=<rgba 0 0 0 0>}
      cornerRadii=<0 0 0 0>
     }
     tintColor {colorUseRule=UseColor_Specified, specifiedColor=<rgba 1 1 1 1>}
     uVRegion {bIsValid=false, max=<0 0>, min=<0 0>}
    }
    padding=<ltrb 4 1 4 2>
   }
  }
  Overlay_0 {widgetClassPath=&Overlay, parent=Border_1, widget=Overlay_0, slot {class=BorderSlot, properties {padding=<ltrb 4 1 4 2>}}}
  Border_0 {
   widgetClassPath=&Border
   parent=Overlay_0
   widget=Border_0
   properties {brushColor=<rgba 0.0097 0.0097 0.0097 1>, padding=<ltrb 4 2 0 1>, renderTransform {angle=0, scale=<2.8 1>, shear=<0 0>, translation=<0 0>}}
   slot {class=OverlaySlot, properties {horizontalAlignment=HAlign_Fill, verticalAlignment=VAlign_Fill, padding=<ltrb 0 0 0 -5>}}
  }
  Image_52 {
   widgetClassPath=&Image
   parent=Border_0
   widget=Image_52
   properties {
    brush {
     bIsDynamicallyLoaded=false
     drawAs=Image
     imageType=NoImage
     mirroring=NoMirror
     resourceName=None
     resourceObject=null
     tiling=NoTile
     imageSize=<5 5>
     margin=<ltrb 0 0 0 0>
     outlineSettings {
      bUseBrushTransparency=false
      roundingType=HalfHeightRadius
      width=0
      color {colorUseRule=UseColor_Specified, specifiedColor=<rgba 0 0 0 0>}
      cornerRadii=<0 0 0 0>
     }
     tintColor {colorUseRule=UseColor_Specified, specifiedColor=<rgba 1 1 1 1>}
     uVRegion {bIsValid=false, max=<0 0>, min=<0 0>}
    }
    colorAndOpacity=<rgba 0.0176 0.0176 0.0176 1>
   }
   slot {class=BorderSlot, properties {padding=<ltrb 4 2 0 1>}}
  }
  HorizontalBox_0 {
   widgetClassPath=&HorizontalBox
   parent=Overlay_0
   widget=HorizontalBox_0
   slot {class=OverlaySlot, properties {horizontalAlignment=HAlign_Fill, padding=<ltrb 0 5 0 5>}}
  }
  SizeBox_0 {
   widgetClassPath=&SizeBox
   parent=HorizontalBox_0
   widget=SizeBox_0
   properties {bOverride_WidthOverride=true, clipping=ClipToBounds, maxDesiredWidth=290, minDesiredWidth=296, widthOverride=296}
   slot {class=HorizontalBoxSlot, properties {verticalAlignment=VAlign_Center, padding=<ltrb 5 0 0 0>, size {sizeRule=Fill, value=1}}}
  }
  PropertyNameText {
   widgetClassPath=&TextBlock
   bIsVariable=true
   parent=SizeBox_0
   widget=PropertyNameText
   properties {
    text=Text Block
    font {
     bForceMonospaced=false
     bMaterialIsStencil=false
     fontMaterial=null
     fontObject=&HeadingNow
     letterSpacing=0
     monospacedWidth=1
     size=7.5
     skewAmount=0
     typefaceFontName=74 Regular
     outlineSettings {
      bApplyOutlineToDropShadows=false
      bMiteredCorners=false
      bSeparateFillAlpha=false
      outlineMaterial=null
      outlineSize=0
      outlineColor=<rgba 0 0 0 1>
     }
    }
   }
   slot {class=SizeBoxSlot}
  }
  Image_95 {
   widgetClassPath=&Image
   parent=HorizontalBox_0
   widget=Image_95
   properties {
    brush {
     bIsDynamicallyLoaded=false
     drawAs=Image
     imageType=NoImage
     mirroring=NoMirror
     resourceName=None
     resourceObject=null
     tiling=NoTile
     imageSize=<2 32>
     margin=<ltrb 0 0 0 0>
     outlineSettings {
      bUseBrushTransparency=false
      roundingType=HalfHeightRadius
      width=0
      color {colorUseRule=UseColor_Specified, specifiedColor=<rgba 0 0 0 0>}
      cornerRadii=<0 0 0 0>
     }
     tintColor {colorUseRule=UseColor_Specified, specifiedColor=<rgba 1 1 1 1>}
     uVRegion {bIsValid=false, max=<0 0>, min=<0 0>}
    }
    colorAndOpacity=<rgba 0.0103 0.0103 0.0103 1>
   }
   slot {class=HorizontalBoxSlot, properties {horizontalAlignment=HAlign_Left, padding=<ltrb 10 -5 0 -10>}}
  }
  SizeBox_1 {
   widgetClassPath=&SizeBox
   parent=HorizontalBox_0
   widget=SizeBox_1
   properties {bOverride_MinDesiredWidth=true, minDesiredWidth=150}
   slot {class=HorizontalBoxSlot, properties {horizontalAlignment=HAlign_Left, size {sizeRule=Fill, value=1}}}
  }
  SinglePropertyView_36 {
   widgetClassPath=&SinglePropertyView
   parent=SizeBox_1
   widget=SinglePropertyView_36
   properties {nameOverride=" ", propertyName=PlayOnAllFXManagers}
   slot {class=SizeBoxSlot}
  }
  Spacer_37 {
   widgetClassPath=&Spacer
   parent=HorizontalBox_0
   widget=Spacer_37
   properties {size=<460 1>}
   slot {class=HorizontalBoxSlot, properties {size {sizeRule=Fill, value=1}}}
  }
 ]
}
