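# Editor Utility Widget that serves as the property window for a Mojo Fx Manager event data asset in the Sparks Stage Manager tools.
# Init binds property views to the asset, toggling visibility by frequency, FX type and animation.
# Target FxManager IDs become combo-box cells listing level BP_MojoFxManager actors, with duplicate flagging and transacted edits.
# Edits update the linked timeline event bar's size and delay, or refresh the timeline editor.
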EditorUtilityWidgetBlueprint /PilgrimQuickplay/StageManager/NeverCook/EditorUtilities/StageManagerEditor/FxManagerEventWindow/SparksFxManagerEventDataWindow
graphs [
 AddNewFxManagerElement {nodes [
   n1 = AddNewFxManagerElement!() -> (then, Index:int, CurrentFxManagerSelected:String) {is=title K2Node_FunctionEntry}
   n2 = Get AllAvaliableFxManagersNames() -> (AllAvaliableFxManagersNames:Array of Strings)
   > n3 = Create Widget!(
    execute:Exec
    WorldContextObject:Object Reference
    Class:User Widget Class Reference=&/EWBP_FxManagerComboBoxCell.EWBP_FxManagerComboBoxCell_C
    OwningPlayer:Player Controller Object Reference
    Index:int=@n1
    AvaliableFXManagerList:Array of Strings=Select(
     Option 0:Array of Strings=Make Array("[0]":String=@n1.CurrentFxManagerSelected) -> (Array:Array of Strings)
     Option 1:Array of Strings=Get AllAvaliableFxManagersNames()
     Index=Is Not Empty!(TargetArray=@n2) -> (ReturnValue:bool)
    ) -> (ReturnValue:Array of Strings)
    CurrentFxManagerSelected:String=@n1.CurrentFxManagerSelected
   ) -> (then:Exec, ReturnValue:EWBP Fx Manager Combo Box Cell Object Reference)
   > Add!(
    execute
    TargetMap:Map of Integers to EWBP Fx Manager Combo Box Cell Object References=Get FxManagersMap() -> (FxManagersMap:Map of Integers to EWBP Fx Manager Combo Box Cell Object References)
    Key:"Integer (by ref)"=@n1
    Value:"EWBP Fx Manager Combo Box Cell Object Reference (by ref)"=@n3
   ) -> (then) {is=title K2Node_CallFunction}
   > Bind Event to On Deleted(self:EWBP Fx Manager Combo Box Cell Object Reference=@n3, Delegate=Create Event()) {is=title K2Node_AddDelegate}
   > Bind Event to Fx Manager Has Changed(self:EWBP Fx Manager Combo Box Cell Object Reference=@n3, Delegate=Create Event()) {is=title K2Node_AddDelegate}
   > AddContentChild!(
    execute
    self:EWBP Sparks Editor Drop Down Header Object Reference=Get EWBP_SparksEditorDropDownHeader() -> (EWBP_SparksEditorDropDownHeader:EWBP Sparks Editor Drop Down Header Object Reference)
    Content:Widget Object Reference=@n3
   ) -> (then, ReturnValue:Panel Slot Object Reference)
  ]}
 CheckForDuplicatedFxManagerId {nodes [
   CheckForDuplicatedFxManagerId() {is=title K2Node_FunctionEntry}
   > n1 = Values!(execute, TargetMap:Map of Integers to EWBP Fx Manager Combo Box Cell Object References=Get FxManagersMap()) -> (then, Values:Array of EWBP Fx Manager Combo Box Cell Object References)
   > n2 = For Each Loop(Array:Array of EWBP Fx Manager Combo Box Cell Object References=@n1) -> (Array Element:EWBP Fx Manager Combo Box Cell Object Reference)
   > n3 = For Each Loop with Break!(
    Exec:Exec
    Array:Array of Names=Get TargetFxManagerIds!(self:Stage Show Event Object Reference=Get EventDataAsset() -> (EventDataAsset:Mojo Fx Manager Event Data Asset Object Reference)) -> (TargetFxManagerIds:Array of Names)
    Break:Exec=@n8.then
   ) -> (LoopBody:Exec, Array Element:Name, Array Index:int, Completed:Exec)
   Set Amount(execute=@n3.Completed, Amount:int=0) -> (Output_Get:int=0)
   > UpdateCellStatus!(execute, self:EWBP Fx Manager Combo Box Cell Object Reference=@n2, IsDuplicated:bool=Get IsDuplicated() -> (IsDuplicated:bool)) -> (then)
   n4 = Branch(
    execute=@n3.LoopBody
    Condition="Equal Exactly (String)"(
     A="To String (Name)"(InName=@n3)
     B=Get CurrentFxManagerSelected!(self:EWBP Fx Manager Combo Box Cell Object Reference=@n2) -> (CurrentFxManagerSelected:String)
    )
   )
   > n5 = Increment Int(Value=Get Amount() -> (Amount:int=0))
   n6 = "integer > integer"(A=@n5, B=1)
   > n7 = Branch(Condition=@n6)
   > n8 = Set IsDuplicated(IsDuplicated:bool=@n6) -> (Output_Get:bool)
   Set IsDuplicated(execute=@n7.else, IsDuplicated=false)
   Set IsDuplicated(execute=@n4.else, IsDuplicated="integer > integer"(A=Get Amount() -> (Amount=0), B=1))
  ]}
 EventGraph {nodes [
   n1 = "On Property Changed (FrequencyTypeProperty)"!() -> (OutputDelegate, then, PropertyName:Name=None) {is=title K2Node_ComponentBoundEvent}
   n2 = Sequence!(execute=[@n1.then, @n29.then_1]) -> (then_0, then_1, then_2:Exec)
   n3 = Get FrequencyType!(self:Stage Show Event Object Reference=Get EventDataAsset()) -> (FrequencyType:EMidiClockSubdivisionQuantization Enum=Bar)
   SetVisibility(
    execute=@n2.then_1
    self=Get SecondsFrequencyProperty() -> (SecondsFrequencyProperty:EWBP Sparks Editor Single Property View Object Reference)
    InVisibility=Select(
     Option 0:ESlateVisibility Enum=Collapsed
     Option 1:ESlateVisibility Enum=Visible
     Index="Equal (Enum)"(A:EMidiClockSubdivisionQuantization Enum=@n3, B:EMidiClockSubdivisionQuantization Enum=None)
    ) -> (ReturnValue:ESlateVisibility Enum)
   )
   n4 = Get FrequencyType(self=Get EventDataAsset()) -> (FrequencyType=Bar)
   SetVisibility(
    execute=@n2.then_2
    self=Get BeatFrequencyProperty() -> (BeatFrequencyProperty:EWBP Sparks Editor Single Property View Object Reference)
    InVisibility=Select(Option 0=Collapsed, Option 1=Visible, Index="Equal (Enum)"(A=@n4, B=Beat))
   )
   n5 = Get FrequencyType(self=Get EventDataAsset()) -> (FrequencyType=Bar)
   > SetVisibility(
    self=Get MeasureFrequency() -> (MeasureFrequency:EWBP Sparks Editor Single Property View Object Reference)
    InVisibility=Select(Option 0=Collapsed, Option 1=Visible, Index="Equal (Enum)"(A=@n5, B=Bar))
   )
   "On Add New Element (EWBP_SparksEditorDropDownHeader)"() {is=title K2Node_ComponentBoundEvent}
   > n6 = Add(TargetArray:Array of Names=Get TargetFxManagerIds(self=Get EventDataAsset()), NewItem:"Name (by ref)") {is=title K2Node_CallArrayFunction}
   > AddNewFxManagerElement!(execute, self, Index:int=@n6, CurrentFxManagerSelected:String) -> (then) {is=title K2Node_CallFunction}
   "On Property Changed (SecondsFrequencyProperty)"!() -> (OutputDelegate, then, PropertyName:Name=None) {is=title K2Node_ComponentBoundEvent}
   > n7 = Get!(execute:Exec) -> (then:Exec, else:Exec, EventBarReference:EWBP Sparks Sequence Pool Timeline Event Bar Object Reference)
   n8 = GetBarData!(self:EWBP Sparks Sequence Pool Timeline Event Bar Object Reference=@n7) -> (
    StepIndex:int=0
    EventIndex:int=0
    SequenceReference:Mojo Fx Manager Sequence Data Asset Object Reference
    EventReference:Stage Show Event Object Reference
   )
   > Is Valid(InputObject=@n8.EventReference)
   > Set EventSize!(
    execute
    EventSize:float=Get NoteFrequency!(self:Stage Show Event Object Reference=@n8.EventReference) -> (NoteFrequency:double)
    self:EWBP Sparks Sequence Pool Timeline Event Bar Object Reference=@n7
   ) -> (then, Output_Get:float)
   > UpdateBar(self:EWBP Sparks Sequence Pool Timeline Event Bar Object Reference=@n7)
   "On Any Property Change (LoopingSettingsHeader)"!() -> (OutputDelegate, then, PropertyName:Name=None) {is=title K2Node_ComponentBoundEvent}
   > n9 = Get!(execute) -> (then, else, SparksTimelineEditor:Sparks Stage Manager Timeline Editor Object Reference)
   > ForceRefresh(self:Sparks Stage Manager Timeline Editor Object Reference=@n9)
   "On Clear All Elements (EWBP_SparksEditorDropDownHeader)"() {is=title K2Node_ComponentBoundEvent}
   > Clear(TargetArray:Array of Names=Get TargetFxManagerIds(self=Get EventDataAsset()))
   n10 = Find(TargetMap:Map of Integers to EWBP Fx Manager Combo Box Cell Object References=Get FxManagersMap(), Key:"Integer (by ref)") -> (Value:EWBP Fx Manager Combo Box Cell Object Reference)
   > Unbind all Events from Fx Manager Has Changed!(execute:Exec, self:EWBP Fx Manager Combo Box Cell Object Reference=@n10) -> (then:Exec) {is=title K2Node_ClearDelegate}
   > Unbind all Events from On Deleted!(execute:Exec, self:EWBP Fx Manager Combo Box Cell Object Reference=@n10) -> (then:Exec) {is=title K2Node_ClearDelegate}
   > Clear!(execute, TargetMap:Map of Integers to EWBP Fx Manager Combo Box Cell Object References=Get FxManagersMap()) -> (then) {is=title K2Node_CallFunction}
   > ClearContentChildren(self:EWBP Sparks Editor Drop Down Header Object Reference=Get EWBP_SparksEditorDropDownHeader())
   n11 = "On Toggle Pressed (GeneralHeader)"!() -> (OutputDelegate, then, IsOpenned:bool) {is=title K2Node_ComponentBoundEvent}
   > SetVisibility(
    self=Get GeneralGroup() -> (GeneralGroup:Vertical Box Object Reference)
    InVisibility=Select(Option 0=Collapsed, Option 1=Visible, Index=@n11.IsOpenned)
   )
   n12 = "On Toggle Pressed (LoopingSettingsHeader)"!() -> (OutputDelegate, then, IsOpenned:bool) {is=title K2Node_ComponentBoundEvent}
   > SetVisibility(
    self=Get LoopingSettingsProperties() -> (LoopingSettingsProperties:Vertical Box Object Reference)
    InVisibility=Select(Option 0=Collapsed, Option 1=Visible, Index=@n12.IsOpenned)
   )
   n13 = "On Property Changed (FxTypeProperty)"!() -> (OutputDelegate, then, PropertyName:Name=None) {is=title K2Node_ComponentBoundEvent}
   n14 = Get FxType!(self:Stage Show Event Object Reference=Get EventDataAsset()) -> (FxType:EShowFXType Enum=Impulse)
   SetVisibility(
    execute=[@n13.then, @n25.then_0]
    self=Get CustomModeIndexProperty() -> (CustomModeIndexProperty:EWBP Sparks Editor Single Property View Object Reference)
    InVisibility=Select(Option 0=Collapsed, Option 1=Visible, Index="Equal (Enum)"(A:EShowFXType Enum=@n14, B:EShowFXType Enum=CustomMode))
   )
   n15 = "On Property Changed (FxAnimationProperty)"!() -> (OutputDelegate, then, PropertyName:Name=None) {is=title K2Node_ComponentBoundEvent}
   n16 = Sequence(execute=[@n15.then, @n25.then_1])
   n17 = Get FxAnimation!(self:Stage Show Event Object Reference=Get EventDataAsset()) -> (FxAnimation:EShowFXAnimation Enum=FireAll)
   > SetVisibility(
    self=Get AnimationStepDelayProperty() -> (AnimationStepDelayProperty:EWBP Sparks Editor Single Property View Object Reference)
    InVisibility=Select(Option 0=Visible, Option 1=Collapsed, Index="Equal (Enum)"(A:EShowFXAnimation Enum=@n17, B:EShowFXAnimation Enum=Linear))
   )
   n18 = Get FxAnimation(self=Get EventDataAsset()) -> (FxAnimation=FireAll)
   SetVisibility(
    execute=@n16.then_1
    self=Get IndicesProperty() -> (IndicesProperty:Border Object Reference)
    InVisibility=Select(Option 0=Collapsed, Option 1=Visible, Index="Equal (Enum)"(A=@n18, B=Alternating))
   )
   n19 = Get FxAnimation(self=Get EventDataAsset()) -> (FxAnimation=FireAll)
   SetVisibility(
    execute=@n16.then_2
    self=Get AlternateCountProperty() -> (AlternateCountProperty:EWBP Sparks Editor Single Property View Object Reference)
    InVisibility=Select(Option 0=Collapsed, Option 1=Visible, Index="Equal (Enum)"(A=@n19, B=FireIndices))
   )
   n20 = "On Toggle Pressed (AnimationSettingsHeader)"!() -> (OutputDelegate, then, IsOpenned:bool) {is=title K2Node_ComponentBoundEvent}
   > SetVisibility(
    self=Get AnimationSettingsProperties() -> (AnimationSettingsProperties:Vertical Box Object Reference)
    InVisibility=Select(Option 0=Collapsed, Option 1=Visible, Index=@n20.IsOpenned)
   )
   "On Pressed (OpenEditorWithThemedGroup)"() {is=title K2Node_ComponentBoundEvent}
   n21 = Get EventDataAsset()
   > SyncBrowserToObjects(AssetPaths=Make Array("[0]"=Get Object Path String(Object=@n21)))
   n22 = "On Any Property Change (AnimationSettingsHeader)"!() -> (OutputDelegate, then, PropertyName:Name=None) {is=title K2Node_ComponentBoundEvent}
   > Branch(Condition="Equal (Name)"(A=@n22.PropertyName, B=StartDelay))
   > n23 = Get!(execute) -> (then, else, EventBarReference:EWBP Sparks Sequence Pool Timeline Event Bar Object Reference)
   n24 = GetBarData(self=@n23) -> (StepIndex=0, EventIndex=0)
   > Is Valid(InputObject=@n24.EventReference)
   > Set DelayValue!(
    execute
    DelayValue:float=Get StartDelay!(self:Stage Show Event Object Reference=@n24.EventReference) -> (StartDelay:double)
    self:EWBP Sparks Sequence Pool Timeline Event Bar Object Reference=@n23
   ) -> (then, Output_Get:float)
   > UpdateBar(self=@n23)
   InitAnimationSettings() {is=title K2Node_CustomEvent}
   > n25 = Sequence!(execute) -> (then_0, then_1)
   n26 = Init!() -> (OutputDelegate, then, EventDataAsset:Mojo Fx Manager Event Data Asset Object Reference) {is=title K2Node_CustomEvent}
   > n27 = Set EventDataAsset(EventDataAsset:Mojo Fx Manager Event Data Asset Object Reference=@n26.EventDataAsset) -> (Output_Get:Mojo Fx Manager Event Data Asset Object Reference)
   n28 = Is Valid(exec=[@n36.then, @n27.then], InputObject=@n27)
   SetVisibility(execute=@n28.Is Not Valid, self=Get MainGroup() -> (MainGroup:Editor Utility Scroll Box Object Reference), InVisibility=Collapsed)
   > SetVisibility(self=Get NoEventGroup() -> (NoEventGroup:Overlay Object Reference), InVisibility=Visible)
   SetVisibility(execute=@n28.Is Valid, self=Get MainGroup(), InVisibility=SelfHitTestInvisible)
   > SetVisibility(self=Get NoEventGroup(), InVisibility=Collapsed)
   > n29 = Sequence!(execute) -> (then_0, then_1, then_2:Exec)
   > ClearContentChildren(self=Get EWBP_SparksEditorDropDownHeader())
   > n30 = GetAllActorsOfClass!(
    execute
    WorldContextObject:Object Reference
    ActorClass:Actor Class Reference=/SparksStageShow/Systems/StageManager/FXManager/BP_MojoFxManager.BP_MojoFxManager_C
   ) -> (then, OutActors:Array of BP Mojo Fx Manager Object References)
   > n31 = Set AllAvaliableFxManagers(AllAvaliableFxManagers:Array of BP Mojo Fx Manager Object References=@n30) -> (Output_Get:Array of BP Mojo Fx Manager Object References)
   > n32 = For Each Loop(Array:Array of BP Mojo Fx Manager Object References=@n31) -> (Array Element:BP Mojo Fx Manager Object Reference)
   n33 = For Each Loop(Exec=@n32.Completed, Array:Array of Names=Get TargetFxManagerIds(self=Get EventDataAsset())) -> (Array Element:Name)
   FillTheHeader!(
    execute=@n33.Completed
    self
    Target:EWBP Sparks Editor Drop Down Main Header Object Reference=Get GeneralHeader() -> (GeneralHeader:EWBP Sparks Editor Drop Down Main Header Object Reference)
    PropertiesList:Array of Names=Make Array!("[0]":Name=PlayOnAllFxManagers, "[1]":Name=StopFXManager) -> (Array:Array of Names)
   ) -> (then)
   > FillTheHeader(
    Target=Get LoopingSettingsHeader() -> (LoopingSettingsHeader:EWBP Sparks Editor Drop Down Main Header Object Reference)
    PropertiesList=Make Array!("[0]"=LoopingEvent) -> (Array)
   )
   > FillTheHeader(
    Target=Get AnimationSettingsHeader() -> (AnimationSettingsHeader:EWBP Sparks Editor Drop Down Main Header Object Reference)
    PropertiesList=Make Array!("[0]"=StartDelay, "[1]":Name=Reverse) -> (Array)
   )
   > FillTheHeader(
    Target=Get NiagaraSystemHeader() -> (NiagaraSystemHeader:EWBP Sparks Editor Drop Down Main Header Object Reference)
    PropertiesList=Make Array!("[0]"=NiagaraSystemOverride) -> (Array)
   )
   > FillLoopingGroup()
   > FillAnimationGroup()
   n34 = Get EventDataAsset()
   > "SetText (Text)"(self=Get DataAssetNameText() -> (DataAssetNameText:Text Object Reference), InText="To Text (String)"(InString=GetDisplayName(Object=@n34)))
   AddNewFxManagerElement(execute=@n33.LoopBody, Index=@n33.Array Index, CurrentFxManagerSelected="To String (Name)"(InName=@n33))
   Add Unique(
    execute=@n32.LoopBody
    TargetArray=Get AllAvaliableFxManagersNames()
    NewItem="To String (Name)"(InName=Get FxManagerId!(self:Stage Show Actors Object Reference=@n32) -> (FxManagerId:Name=None))
   )
   n35 = Event PreConstruct()
   > n36 = Branch(Condition=@n35.IsDesignTime)
  ]}
 FillAnimationGroup {nodes [
   n1 = SetObject!(
    execute
    self:EWBP Sparks Editor Single Property View Object Reference=Get ReverseProperty() -> (ReverseProperty:EWBP Sparks Editor Single Property View Object Reference)
    NewObject=Get EventDataAsset()
    NewPropertyName:Name=Reverse
   ) -> (then)
   SetObject(execute=[@n2.then, @n1.then], self=Get AnimationStepDelayProperty(), NewObject=Get EventDataAsset(), NewPropertyName=AnimationStepDelay)
   > SetObject!(
    execute
    self:Property View Base Object Reference=Get FortMultiPropertyView_198() -> (FortMultiPropertyView_198:Fort Multi Property View Object Reference)
    NewObject=Get EventDataAsset()
   ) -> (then)
   > SetObject!(
    execute
    self:EWBP Sparks Editor Single Property View Object Reference=Get AlternateCountProperty()
    NewObject=Get EventDataAsset()
    NewPropertyName:Name=AlternateCount
   ) -> (then)
   > InitAnimationSettings() {is=title K2Node_CallFunction}
   FillAnimationGroup() {is=title K2Node_FunctionEntry}
   > SetObject(self=Get CustomModeIndexProperty(), NewObject=Get EventDataAsset(), NewPropertyName=CustomModeIndex)
   > SetObject(
    self=Get FxAnimationProperty() -> (FxAnimationProperty:EWBP Sparks Editor Single Property View Object Reference)
    NewObject=Get EventDataAsset()
    NewPropertyName=FxAnimation
   )
   > SetObject(
    self=Get FxTypeProperty() -> (FxTypeProperty:EWBP Sparks Editor Single Property View Object Reference)
    NewObject=Get EventDataAsset()
    NewPropertyName=FxType
   )
   > n2 = SetObject(
    self=Get LoopCountProperty() -> (LoopCountProperty:EWBP Sparks Editor Single Property View Object Reference)
    NewObject=Get EventDataAsset()
    NewPropertyName=LoopCount
   )
  ]}
 FillLoopingGroup {nodes [
   FillLoopingGroup() {is=title K2Node_FunctionEntry}
   > SetObject(
    self=Get FrequencyTypeProperty() -> (FrequencyTypeProperty:EWBP Sparks Editor Single Property View Object Reference)
    NewObject=Get EventDataAsset()
    NewPropertyName=FrequencyType
   )
   > SetObject(self=Get SecondsFrequencyProperty(), NewObject=Get EventDataAsset(), NewPropertyName=SecondsFrequency)
   > SetObject(self=Get BeatFrequencyProperty(), NewObject=Get EventDataAsset(), NewPropertyName=Beat Frequency)
   > SetObject(self=Get MeasureFrequency(), NewObject=Get EventDataAsset(), NewPropertyName=Measure Frequency)
   n1 = NOT Boolean(A=Get IsFromTimelineEditor() -> (IsFromTimelineEditor:bool))
   > SetChildPropertyEnabledAt!(
    execute
    self:EWBP Sparks Editor Drop Down Main Header Object Reference=Get LoopingSettingsHeader()
    Index:int=0
    bInIsEnabled:bool=@n1
    InToolTipText:"Text (by ref)"=SelectText(
     B="INVTEXT(\"You can edit this value because the timeline tool already specify it, please keep the atention to the naming of the data asset before changing this value for keep thing organized on the project. If you want to modfy this value do it opening this window outside the tool.\")"
     bPickA=@n1
    )
   ) -> (then)
  ]}
 FillTheHeader {nodes [
   n1 = FillTheHeader!() -> (then, Target:EWBP Sparks Editor Drop Down Main Header Object Reference, PropertiesList:Array of Names) {is=title K2Node_FunctionEntry}
   > ClearContentChildren(self:EWBP Sparks Editor Drop Down Main Header Object Reference=@n1)
   > n2 = For Each Loop(Array=@n1.PropertiesList)
   > n3 = Create Widget!(
    execute
    WorldContextObject
    Class=&/EWBP_SparksEditorSinglePropertyView.EWBP_SparksEditorSinglePropertyView_C
    OwningPlayer
   ) -> (then, ReturnValue:EWBP Sparks Editor Single Property View Object Reference)
   > SetObject(self=@n3, NewObject=Get EventDataAsset(), NewPropertyName=@n2)
   > AddContentChild(self:EWBP Sparks Editor Drop Down Main Header Object Reference=@n1, Content=@n3)
  ]}
 OnFxManagerCellDataChanged {nodes [
   n1 = OnFxManagerCellDataChanged!() -> (then, Index:int, SelectedFxManager:String) {is=title K2Node_FunctionEntry}
   > n2 = Sequence!(execute) -> (then_0, then_1)
   n3 = String To Name(InString=@n1.SelectedFxManager)
   n4 = Get TargetFxManagerIds(self=Get EventDataAsset())
   > Branch(Condition="Not Equal (Name)"(A="Get (a ref)"(Array:Array of Names=@n4, Dimension 1=@n1) -> (Output:"Name (by ref)"), B=@n3) -> (ReturnValue=~))
   > BeginTransaction(Context=ChangeFxManagerCellValue, Description="INVTEXT(\"Change the FXManager index value.\")", PrimaryObject=Get EventDataAsset())
   n5 = Get EventDataAsset()
   > "Set Name (by ref)"!(execute:Exec, Target:"Name (by ref)"="Get (a ref)"(Array=Get TargetFxManagerIds(self=@n5), Dimension 1=@n1), Value:Name=@n3) -> (then:Exec) {is=title K2Node_VariableSetRef}
   > EndTransaction()
   > TransactObject(Object=Get EventDataAsset())
   CheckForDuplicatedFxManagerId(execute=@n2.then_1) {is=title K2Node_CallFunction}
   > n6 = Get()
   > CheckForDuplicatedManagers(self:EWBP Sparks Sequence Pool Timeline Event Bar Object Reference=@n6)
  ]}
 OnFxManagerCellDeleted {nodes [
   n1 = OnFxManagerCellDeleted!() -> (then, Index:int) {is=title K2Node_FunctionEntry}
   > BeginTransaction(Context=RemovingFxmanagerCell, Description="INVTEXT(\"Remove the FXManager Cell\")", PrimaryObject=Get EventDataAsset())
   n2 = Find(TargetMap=Get FxManagersMap(), Key=@n1)
   > Unbind Event from On Deleted(self:EWBP Fx Manager Combo Box Cell Object Reference=@n2, Delegate=Create Event()) {is=title K2Node_RemoveDelegate}
   > Unbind Event from Fx Manager Has Changed(self:EWBP Fx Manager Combo Box Cell Object Reference=@n2, Delegate=Create Event()) {is=title K2Node_RemoveDelegate}
   > RemoveContentChild!(
    execute
    self:EWBP Sparks Editor Drop Down Header Object Reference=Get EWBP_SparksEditorDropDownHeader()
    Content:Widget Object Reference=@n2
   ) -> (then, ReturnValue:bool)
   > Remove!(execute, TargetMap:Map of Integers to EWBP Fx Manager Combo Box Cell Object References=Get FxManagersMap(), Key:"Integer (by ref)"=@n1) -> (then, ReturnValue:bool)
   > Remove Index!(execute, TargetArray:Array of Names=Get TargetFxManagerIds(self=Get EventDataAsset()), IndexToRemove:int=@n1) -> (then)
   > EndTransaction()
   > TransactObject(Object=Get EventDataAsset())
   > CheckForDuplicatedFxManagerId()
   > n3 = Get()
   > CheckForDuplicatedManagers(self=@n3)
  ]}
 SetEventBarReference {nodes [
   n1 = SetEventBarReference!() -> (then, EventBarReference:"EWBP Sparks Sequence Pool Timeline Event Bar Object Reference (by ref)") {is=title K2Node_FunctionEntry}
   > Set EventBarReference(EventBarReference:EWBP Sparks Sequence Pool Timeline Event Bar Object Reference=@n1) -> (Output_Get:EWBP Sparks Sequence Pool Timeline Event Bar Object Reference)
  ]}
 SetTimelineEditorReference {nodes [
   n1 = SetTimelineEditorReference!() -> (then, SparksTimelineEditor:Sparks Stage Manager Timeline Editor Object Reference) {is=title K2Node_FunctionEntry}
   > Set SparksTimelineEditor(SparksTimelineEditor:Sparks Stage Manager Timeline Editor Object Reference=@n1) -> (Output_Get:Sparks Stage Manager Timeline Editor Object Reference)
  ]}
]
variables [
 AllAvaliableFxManagers {type="array<BP_MojoFxManager_C>", default []}
 AllAvaliableFxManagersNames {type="array<string>", default []}
 EventBarReference {type=EWBP_SparksSequencePool_TimelineEventBar_C, default=null}
 EventDataAsset {type=MojoFxManagerEvent_DataAsset_C, default=/PilgrimQuickplay/StageManager/Data/Events/Pilgrim_StageFX_Event_Intro02_ColorShift}
 FxManagersMap {default {}}
 IsFromTimelineEditor {type=boolean, default=true}
 SparksTimelineEditor {type=SparksStageManagerTimelineEditor_C, default=null}
]
widgets {
 info {rootWidgetClass=&Overlay, widgetCount=45}
 widgets [
  Overlay_0 {widgetClassPath=&Overlay, widget=Overlay_0}
  MainGroup {
   widgetClassPath=&EditorUtilityScrollBox
   bIsVariable=true
   parent=Overlay_0
   widget=MainGroup
   properties {visibility=Collapsed}
   slot {class=OverlaySlot, properties {horizontalAlignment=HAlign_Fill, padding=<ltrb 0 5 0 0>}}
  }
  VerticalBox_0 {widgetClassPath=&VerticalBox, parent=MainGroup, widget=VerticalBox_0, slot {class=ScrollBoxSlot}}
  Overlay_1 {
   widgetClassPath=&Overlay
   parent=VerticalBox_0
   widget=Overlay_1
   slot {class=VerticalBoxSlot, properties {padding=<ltrb 15 10 15 2>}}
  }
  Image {
   widgetClassPath=&Image
   parent=Overlay_1
   widget=Image
   properties {
    brush {
     bIsDynamicallyLoaded=false
     drawAs=Image
     imageType=NoImage
     mirroring=NoMirror
     resourceName=None
     resourceObject=null
     tiling=NoTile
     imageSize=<32 32>
     margin=<ltrb 0 0 0 0>
     outlineSettings {
      bUseBrushTransparency=false
      roundingType=HalfHeightRadius
      width=0
      color {colorUseRule=UseColor_Specified, specifiedColor=<rgba 0 0 0 0>}
      cornerRadii=<0 0 0 0>
     }
     tintColor {colorUseRule=UseColor_Specified, specifiedColor=<rgba 0.0097 0.0097 0.0097 1>}
     uVRegion {bIsValid=false, max=<0 0>, min=<0 0>}
    }
    renderTransform {angle=0, scale=<1 1>, shear=<0 0>, translation=<0 2>}
   }
   slot {class=OverlaySlot, properties {horizontalAlignment=HAlign_Fill, verticalAlignment=VAlign_Fill, padding=<ltrb -30 0 -30 0>}}
  }
  Image_29 {
   widgetClassPath=&Image
   parent=Overlay_1
   widget=Image_29
   properties {brush {
     bIsDynamicallyLoaded=false
     drawAs=Image
     imageType=NoImage
     mirroring=NoMirror
     resourceName=None
     resourceObject=null
     tiling=NoTile
     imageSize=<32 32>
     margin=<ltrb 0 0 0 0>
     outlineSettings {
      bUseBrushTransparency=false
      roundingType=HalfHeightRadius
      width=0
      color {colorUseRule=UseColor_Specified, specifiedColor=<rgba 0 0 0 0>}
      cornerRadii=<0 0 0 0>
     }
     tintColor {colorUseRule=UseColor_Specified, specifiedColor=<rgba 0.0232 0.0232 0.0232 1>}
     uVRegion {bIsValid=false, max=<0 0>, min=<0 0>}
    }}
   slot {class=OverlaySlot, properties {horizontalAlignment=HAlign_Fill, verticalAlignment=VAlign_Fill, padding=<ltrb -30 0 -30 0>}}
  }
  HorizontalBox_1 {
   widgetClassPath=&HorizontalBox
   parent=Overlay_1
   widget=HorizontalBox_1
   slot {class=OverlaySlot, properties {horizontalAlignment=HAlign_Fill, verticalAlignment=VAlign_Center}}
  }
  SizeBox_2 {
   widgetClassPath=&SizeBox
   parent=HorizontalBox_1
   widget=SizeBox_2
   properties {minDesiredWidth=360}
   slot {class=HorizontalBoxSlot}
  }
  DataAssetNameText {
   widgetClassPath=&TextBlock
   bIsVariable=true
   parent=SizeBox_2
   widget=DataAssetNameText
   properties {
    text=Data Asset Name
    font {
     bForceMonospaced=false
     bMaterialIsStencil=false
     fontMaterial=null
     fontObject=&HeadingNow
     letterSpacing=0
     monospacedWidth=1
     size=11.25
     skewAmount=0
     typefaceFontName=75 Medium
     outlineSettings {
      bApplyOutlineToDropShadows=false
      bMiteredCorners=false
      bSeparateFillAlpha=false
      outlineMaterial=null
      outlineSize=0
      outlineColor=<rgba 0 0 0 1>
     }
    }
   }
   slot {class=SizeBoxSlot}
  }
  OpenEditorWithThemedGroup {
   widgetClassPath=&EditorUtilityButton
   parent=HorizontalBox_1
   widget=OpenEditorWithThemedGroup
   properties {
    toolTipText=Browse to asset in Content Browser
    widgetStyle {
     clickedSlateSound {resourceObject=null}
     disabled {
      bIsDynamicallyLoaded=false
      drawAs=RoundedBox
      imageType=NoImage
      mirroring=NoMirror
      resourceName=None
      resourceObject=/Game/UI/Foundation/Textures/Buttons/T-Button-Square
      tiling=NoTile
      imageSize=<32 32>
      margin=<ltrb 0.5 0.5 0.5 0.5>
      outlineSettings {
       bUseBrushTransparency=false
       roundingType=FixedRadius
       width=0
       color {colorUseRule=UseColor_ColorTable, specifiedColor=<rgba 0 0 0 0>}
       cornerRadii=<4 4 4 4>
      }
      tintColor {colorUseRule=UseColor_Specified, specifiedColor=<rgba 0.0395 0.0395 0.0395 1>}
      uVRegion {bIsValid=false, max=<0 0>, min=<0 0>}
     }
     disabledForeground {colorUseRule=UseColor_ColorTable, specifiedColor=<rgba 0 0 0 0>}
     hovered {
      bIsDynamicallyLoaded=false
      drawAs=RoundedBox
      imageType=NoImage
      mirroring=NoMirror
      resourceName=None
      resourceObject=&T-Button-Square
      tiling=NoTile
      imageSize=<32 32>
      margin=<ltrb 0.5 0.5 0.5 0.5>
      outlineSettings {
       bUseBrushTransparency=false
       roundingType=FixedRadius
       width=0
       color {colorUseRule=UseColor_ColorTable, specifiedColor=<rgba 0 0 0 0>}
       cornerRadii=<4 4 4 4>
      }
      tintColor {colorUseRule=UseColor_Specified, specifiedColor=<rgba 0 0.3021 0.0241 1>}
      uVRegion {bIsValid=false, max=<0 0>, min=<0 0>}
     }
     hoveredForeground {colorUseRule=UseColor_ColorTable, specifiedColor=<rgba 0 0 0 0>}
     hoveredSlateSound {resourceObject=null}
     normal {
      bIsDynamicallyLoaded=false
      drawAs=RoundedBox
      imageType=NoImage
      mirroring=NoMirror
      resourceName=None
      resourceObject=&T-Button-Square
      tiling=NoTile
      imageSize=<32 32>
      margin=<ltrb 0.5 0.5 0.5 0.5>
      outlineSettings {
       bUseBrushTransparency=false
       roundingType=FixedRadius
       width=0
       color {colorUseRule=UseColor_ColorTable, specifiedColor=<rgba 0 0 0 0>}
       cornerRadii=<4 4 4 4>
      }
      tintColor {colorUseRule=UseColor_Specified, specifiedColor=<rgba 0.0071 0.7594 0 1>}
      uVRegion {bIsValid=false, max=<0 0>, min=<0 0>}
     }
     normalForeground {colorUseRule=UseColor_ColorTable, specifiedColor=<rgba 0 0 0 0>}
     normalPadding=<ltrb 12 1.5 12 1.5>
     pressed {
      bIsDynamicallyLoaded=false
      drawAs=RoundedBox
      imageType=NoImage
      mirroring=NoMirror
      resourceName=None
      resourceObject=&T-Button-Square
      tiling=NoTile
      imageSize=<32 32>
      margin=<ltrb 0.5 0.5 0.5 0.5>
      outlineSettings {
       bUseBrushTransparency=false
       roundingType=FixedRadius
       width=0
       color {colorUseRule=UseColor_ColorTable, specifiedColor=<rgba 0 0 0 0>}
       cornerRadii=<4 4 4 4>
      }
      tintColor {colorUseRule=UseColor_Specified, specifiedColor=<rgba 0.3119 0.7917 0.2715 1>}
      uVRegion {bIsValid=false, max=<0 0>, min=<0 0>}
     }
     pressedForeground {colorUseRule=UseColor_ColorTable, specifiedColor=<rgba 0 0 0 0>}
     pressedPadding=<ltrb 12 2.5 12 0.5>
     pressedSlateSound {resourceObject=null}
    }
   }
   slot {class=HorizontalBoxSlot, properties {horizontalAlignment=HAlign_Left, padding=<ltrb 10 0 0 0>}}
  }
  ExternalIcon {
   widgetClassPath=&Image
   parent=OpenEditorWithThemedGroup
   widget=ExternalIcon
   properties {
    brush {
     bIsDynamicallyLoaded=false
     drawAs=Image
     imageType=FullColor
     mirroring=NoMirror
     resourceName=None
     resourceObject=/PilgrimQuickplay/StageManager/NeverCook/EditorUtilities/StageManagerEditor/Textures/T_SparksEditorUI_IconMagnifying_Glass_32
     tiling=NoTile
     imageSize=<20 20>
     margin=<ltrb 0 0 0 0>
     outlineSettings {
      bUseBrushTransparency=false
      roundingType=HalfHeightRadius
      width=0
      color {colorUseRule=UseColor_Specified, specifiedColor=<rgba 0 0 0 0>}
      cornerRadii=<0 0 0 0>
     }
     tintColor {colorUseRule=UseColor_Specified, specifiedColor=<rgba 1 1 1 1>}
     uVRegion {bIsValid=false, max=<0 0>, min=<0 0>}
    }
    renderTransform {angle=0, scale=<-1 1>, shear=<0 0>, translation=<0 0>}
   }
   slot {class=ButtonSlot, properties {padding=<ltrb -2 0 -2 0>}}
  }
  GeneralVGroup {widgetClassPath=&VerticalBox, parent=VerticalBox_0, widget=GeneralVGroup, slot {class=VerticalBoxSlot}}
  GeneralHeader {
   widgetClassPath=&/EWBP_SparksEditorDropDownMainHeader
   bIsVariable=true
   parent=GeneralVGroup
   widget=GeneralHeader
   properties {headerTitle=General}
   slot {class=VerticalBoxSlot, properties {verticalAlignment=VAlign_Top}}
  }
  GeneralGroup {widgetClassPath=&VerticalBox, bIsVariable=true, parent=GeneralVGroup, widget=GeneralGroup, slot {class=VerticalBoxSlot}}
  EWBP_SparksEditorDropDownHeader {
   widgetClassPath=&/EWBP_SparksEditorDropDownHeader
   bIsVariable=true
   parent=GeneralGroup
   widget=EWBP_SparksEditorDropDownHeader
   properties {headerTitle=Target Fx Manager Ids}
   slot {class=VerticalBoxSlot, properties {verticalAlignment=VAlign_Top}}
  }
  LoopingSettingsVGroup {
   widgetClassPath=&VerticalBox
   parent=VerticalBox_0
   widget=LoopingSettingsVGroup
   slot {class=VerticalBoxSlot, properties {verticalAlignment=VAlign_Top}}
  }
  LoopingSettingsHeader {
   widgetClassPath=&EWBP_SparksEditorDropDownMainHeader
   bIsVariable=true
   parent=LoopingSettingsVGroup
   widget=LoopingSettingsHeader
   properties {headerTitle=Looping Settings}
   slot {class=VerticalBoxSlot}
  }
  LoopingSettingsProperties {
   widgetClassPath=&VerticalBox
   bIsVariable=true
   parent=LoopingSettingsVGroup
   widget=LoopingSettingsProperties
   slot {class=VerticalBoxSlot, properties {padding=<ltrb 24 0 0 0>}}
  }
  FrequencyTypeProperty {
   widgetClassPath=&/EWBP_SparksEditorSinglePropertyView
   bIsVariable=true
   parent=LoopingSettingsProperties
   widget=FrequencyTypeProperty
   slot {class=VerticalBoxSlot}
  }
  SecondsFrequencyProperty {
   widgetClassPath=&EWBP_SparksEditorSinglePropertyView
   bIsVariable=true
   parent=LoopingSettingsProperties
   widget=SecondsFrequencyProperty
   slot {class=VerticalBoxSlot}
  }
  BeatFrequencyProperty {
   widgetClassPath=&EWBP_SparksEditorSinglePropertyView
   bIsVariable=true
   parent=LoopingSettingsProperties
   widget=BeatFrequencyProperty
   slot {class=VerticalBoxSlot}
  }
  MeasureFrequency {
   widgetClassPath=&EWBP_SparksEditorSinglePropertyView
   bIsVariable=true
   parent=LoopingSettingsProperties
   widget=MeasureFrequency
   slot {class=VerticalBoxSlot}
  }
  AnimationSettingsVGroup {
   widgetClassPath=&VerticalBox
   parent=VerticalBox_0
   widget=AnimationSettingsVGroup
   slot {class=VerticalBoxSlot, properties {verticalAlignment=VAlign_Top}}
  }
  AnimationSettingsHeader {
   widgetClassPath=&EWBP_SparksEditorDropDownMainHeader
   bIsVariable=true
   parent=AnimationSettingsVGroup
   widget=AnimationSettingsHeader
   properties {headerTitle=Animation Settings}
   slot {class=VerticalBoxSlot, properties {padding=<ltrb 0 3 0 0>}}
  }
  AnimationSettingsProperties {
   widgetClassPath=&VerticalBox
   bIsVariable=true
   parent=AnimationSettingsVGroup
   widget=AnimationSettingsProperties
   slot {class=VerticalBoxSlot, properties {padding=<ltrb 24 0 0 0>}}
  }
  FxTypeProperty {
   widgetClassPath=&EWBP_SparksEditorSinglePropertyView
   bIsVariable=true
   parent=AnimationSettingsProperties
   widget=FxTypeProperty
   slot {class=VerticalBoxSlot}
  }
  FxAnimationProperty {
   widgetClassPath=&EWBP_SparksEditorSinglePropertyView
   bIsVariable=true
   parent=AnimationSettingsProperties
   widget=FxAnimationProperty
   slot {class=VerticalBoxSlot}
  }
  LoopCountProperty {
   widgetClassPath=&EWBP_SparksEditorSinglePropertyView
   bIsVariable=true
   parent=AnimationSettingsProperties
   widget=LoopCountProperty
   slot {class=VerticalBoxSlot}
  }
  AnimationStepDelayProperty {
   widgetClassPath=&EWBP_SparksEditorSinglePropertyView
   bIsVariable=true
   parent=AnimationSettingsProperties
   widget=AnimationStepDelayProperty
   slot {class=VerticalBoxSlot}
  }
  CustomModeIndexProperty {
   widgetClassPath=&EWBP_SparksEditorSinglePropertyView
   bIsVariable=true
   parent=AnimationSettingsProperties
   widget=CustomModeIndexProperty
   slot {class=VerticalBoxSlot}
  }
  AlternateCountProperty {
   widgetClassPath=&EWBP_SparksEditorSinglePropertyView
   bIsVariable=true
   parent=AnimationSettingsProperties
   widget=AlternateCountProperty
   slot {class=VerticalBoxSlot}
  }
  IndicesProperty {
   widgetClassPath=&Border
   bIsVariable=true
   parent=AnimationSettingsProperties
   widget=IndicesProperty
   properties {
    background {
     bIsDynamicallyLoaded=false
     drawAs=NoDrawType
     imageType=NoImage
     mirroring=NoMirror
     resourceName=None
     resourceObject=null
     tiling=NoTile
     imageSize=<32 32>
     margin=<ltrb 0 0 0 0>
     outlineSettings {
      bUseBrushTransparency=false
      roundingType=HalfHeightRadius
      width=0
      color {colorUseRule=UseColor_Specified, specifiedColor=<rgba 0 0 0 0>}
      cornerRadii=<0 0 0 0>
     }
     tintColor {colorUseRule=UseColor_Specified, specifiedColor=<rgba 1 1 1 1>}
     uVRegion {bIsValid=false, max=<0 0>, min=<0 0>}
    }
    padding=<ltrb 4 1 4 2>
   }
   slot {class=VerticalBoxSlot}
  }
  Overlay {widgetClassPath=&Overlay, parent=IndicesProperty, widget=Overlay, slot {class=BorderSlot, properties {padding=<ltrb 4 1 4 2>}}}
  Border_0 {
   widgetClassPath=&Border
   parent=Overlay
   widget=Border_0
   properties {brushColor=<rgba 0.0097 0.0097 0.0097 1>, padding=<ltrb 4 2 0 1>, renderTransform {angle=0, scale=<2.8 1>, shear=<0 0>, translation=<0 0>}}
   slot {class=OverlaySlot, properties {horizontalAlignment=HAlign_Fill, verticalAlignment=VAlign_Fill, padding=<ltrb 0 0 0 -5>}}
  }
  Image_52 {
   widgetClassPath=&Image
   parent=Border_0
   widget=Image_52
   properties {
    brush {
     bIsDynamicallyLoaded=false
     drawAs=Image
     imageType=NoImage
     mirroring=NoMirror
     resourceName=None
     resourceObject=null
     tiling=NoTile
     imageSize=<5 5>
     margin=<ltrb 0 0 0 0>
     outlineSettings {
      bUseBrushTransparency=false
      roundingType=HalfHeightRadius
      width=0
      color {colorUseRule=UseColor_Specified, specifiedColor=<rgba 0 0 0 0>}
      cornerRadii=<0 0 0 0>
     }
     tintColor {colorUseRule=UseColor_Specified, specifiedColor=<rgba 1 1 1 1>}
     uVRegion {bIsValid=false, max=<0 0>, min=<0 0>}
    }
    colorAndOpacity=<rgba 0.0176 0.0176 0.0176 1>
   }
   slot {class=BorderSlot, properties {padding=<ltrb 4 2 0 1>}}
  }
  HorizontalBox_0 {
   widgetClassPath=&HorizontalBox
   parent=Overlay
   widget=HorizontalBox_0
   slot {class=OverlaySlot, properties {padding=<ltrb 0 5 0 5>}}
  }
  SizeBox_0 {
   widgetClassPath=&SizeBox
   parent=HorizontalBox_0
   widget=SizeBox_0
   properties {bOverride_MinDesiredWidth=true, minDesiredWidth=300}
   slot {class=HorizontalBoxSlot, properties {verticalAlignment=VAlign_Center, padding=<ltrb 5 0 0 0>}}
  }
  PropertyNameText {
   widgetClassPath=&TextBlock
   bIsVariable=true
   parent=SizeBox_0
   widget=PropertyNameText
   properties {
    text=Indices
    font {
     bForceMonospaced=false
     bMaterialIsStencil=false
     fontMaterial=null
     fontObject=&HeadingNow
     letterSpacing=0
     monospacedWidth=1
     size=7.5
     skewAmount=0
     typefaceFontName=74 Regular
     outlineSettings {
      bApplyOutlineToDropShadows=false
      bMiteredCorners=false
      bSeparateFillAlpha=false
      outlineMaterial=null
      outlineSize=0
      outlineColor=<rgba 0 0 0 1>
     }
    }
   }
   slot {class=SizeBoxSlot}
  }
  Image_95 {
   widgetClassPath=&Image
   parent=HorizontalBox_0
   widget=Image_95
   properties {
    brush {
     bIsDynamicallyLoaded=false
     drawAs=Image
     imageType=NoImage
     mirroring=NoMirror
     resourceName=None
     resourceObject=null
     tiling=NoTile
     imageSize=<2 32>
     margin=<ltrb 0 0 0 0>
     outlineSettings {
      bUseBrushTransparency=false
      roundingType=HalfHeightRadius
      width=0
      color {colorUseRule=UseColor_Specified, specifiedColor=<rgba 0 0 0 0>}
      cornerRadii=<0 0 0 0>
     }
     tintColor {colorUseRule=UseColor_Specified, specifiedColor=<rgba 1 1 1 1>}
     uVRegion {bIsValid=false, max=<0 0>, min=<0 0>}
    }
    colorAndOpacity=<rgba 0.0103 0.0103 0.0103 1>
   }
   slot {class=HorizontalBoxSlot, properties {padding=<ltrb 10 -5 0 -10>}}
  }
  SizeBox_1 {
   widgetClassPath=&SizeBox
   parent=HorizontalBox_0
   widget=SizeBox_1
   properties {bOverride_MinDesiredWidth=true, minDesiredWidth=150}
   slot {class=HorizontalBoxSlot, properties {size {sizeRule=Fill, value=1}}}
  }
  FortMultiPropertyView_198 {
   widgetClassPath=/Script/FortniteEditor.FortMultiPropertyView
   bIsVariable=true
   parent=SizeBox_1
   widget=FortMultiPropertyView_198
   properties {bShowNameWidgets=false, propertiesToShow [Indices]}
   slot {class=SizeBoxSlot, properties {padding=<ltrb 15 0 0 0>}}
  }
  ReverseProperty {
   widgetClassPath=&EWBP_SparksEditorSinglePropertyView
   bIsVariable=true
   parent=AnimationSettingsProperties
   widget=ReverseProperty
   properties {visibility=Collapsed}
   slot {class=VerticalBoxSlot}
  }
  NiagaraSystemHeader {
   widgetClassPath=&EWBP_SparksEditorDropDownMainHeader
   bIsVariable=true
   parent=VerticalBox_0
   widget=NiagaraSystemHeader
   properties {headerTitle=Niagara System Particles}
   slot {class=VerticalBoxSlot, properties {padding=<ltrb 0 3 0 0>}}
  }
  NoEventGroup {
   widgetClassPath=&Overlay
   bIsVariable=true
   parent=Overlay_0
   widget=NoEventGroup
   properties {visibility=Visible}
   slot {class=OverlaySlot, properties {horizontalAlignment=HAlign_Fill, verticalAlignment=VAlign_Fill}}
  }
  NoEventText {
   widgetClassPath=&TextBlock
   parent=NoEventGroup
   widget=NoEventText
   properties {
    justification=Center
    text=No Event Data Asset Selected Please open using some sparks tool.
    font {
     bForceMonospaced=false
     bMaterialIsStencil=false
     fontMaterial=null
     fontObject=&HeadingNow
     letterSpacing=0
     monospacedWidth=1
     size=11.25
     skewAmount=0
     typefaceFontName=75 Medium
     outlineSettings {
      bApplyOutlineToDropShadows=false
      bMiteredCorners=false
      bSeparateFillAlpha=false
      outlineMaterial=null
      outlineSize=0
      outlineColor=<rgba 0 0 0 1>
     }
    }
   }
   slot {class=OverlaySlot, properties {horizontalAlignment=HAlign_Fill, verticalAlignment=VAlign_Center}}
  }
 ]
}
