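# Editor utility action in the Stage Manager tools that opens the FX Manager event data window for a selected event asset.
# Its single function spawns and registers the SparksFxManagerEventDataWindow widget tab through the Editor Utility Subsystem.
# It casts the first selected asset to MojoFxManagerEvent_DataAsset, flags the window as not opened from the timeline editor, then initializes it with that asset.
# The tab spawns before the asset type check, and failed casts have no handling.
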EditorUtilityBlueprint /PilgrimQuickplay/StageManager/NeverCook/EditorUtilities/StageManagerEditor/FxManagerEventWindow/EBP_OpenCustomEventWindowAction
graphs [
 EventGraph {links [], nodes []}
 Open Custom Event Editor Window {nodes [
   Open Custom Event Editor Window() {is=title K2Node_FunctionEntry}
   > n1 = SpawnAndRegisterTabAndGetID!(
    execute
    self:Editor Utility Subsystem Object Reference=Get EditorUtilitySubsystem()
    InBlueprint:Editor Utility Widget Blueprint Object Reference=&/SparksFxManagerEventDataWindow
   ) -> (then, NewTabID:Name=None, ReturnValue:Editor Utility Widget Object Reference)
   > n2 = Cast To SparksFxManagerEventDataWindow!(execute, Object=@n1.ReturnValue) -> (then, CastFailed, AsSparks Fx Manager Event Data Window:Sparks Fx Manager Event Data Window Object Reference) {is=title K2Node_DynamicCast}
   > n3 = GetSelectedAssets()
   > n4 = Cast To MojoFxManagerEvent_DataAsset!(execute, Object="Get (a copy)"(Array:Array of Object References=@n3) -> (Output:Object Reference)) -> (then, CastFailed, AsMojo Fx Manager Event Data Asset:Mojo Fx Manager Event Data Asset Object Reference) {is=title K2Node_DynamicCast}
   > Set IsFromTimelineEditor!(execute, IsFromTimelineEditor:bool=false, self:Sparks Fx Manager Event Data Window Object Reference=@n2) -> (then, Output_Get:bool)
   > Init!(execute, self:Sparks Fx Manager Event Data Window Object Reference=@n2, EventDataAsset:Mojo Fx Manager Event Data Asset Object Reference=@n4) -> (then)
  ]}
]
variables []
